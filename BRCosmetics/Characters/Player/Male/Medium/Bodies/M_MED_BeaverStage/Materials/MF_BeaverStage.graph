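# Material function for the BeaverStage character body that adds helmet shadowing to incoming material attributes through an ambient occlusion mask.
# A selectable TechArtMask channel is inverted, shaped by power and bias, then clamped into the AO mask.
# The mask drives Ambient Occlusion, is exposed as a second output, and controls BaseColor darkening.
# A quality switch applies separate darkening intensities for low-end/mobile and High/Epic tiers, with the latter defaulting to none.
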MaterialFunction /BRCosmetics/Characters/Player/Male/Medium/Bodies/M_MED_BeaverStage/Materials/MF_BeaverStage
graph {nodes [
  n1 = GetMaterialAttributes(None=FunctionInput() {inputType=FunctionInput_MaterialAttributes}) {attributeGetTypes [#1]}
  n2 = ScalarParameter() {defaultValue=0.8, group=BeaverStage - Helmet Shadow, parameterName=Low End - AO Intensity}
  n3 = LinearInterpolate(A=Multiply(A=@n1.BaseColor, B=OneMinus(None=@n2)) {constB=0.5}, B=@n1.BaseColor, Alpha=NamedRerouteUsage())
  n4 = ScalarParameter() {group=BeaverStage - Helmet Shadow, parameterName=Frontend - AO Base Color Intensity}
  n5 = LinearInterpolate(A=Multiply(A=@n1.BaseColor, B=OneMinus(None=@n4)), B=@n1.BaseColor, Alpha=NamedRerouteUsage())
  FunctionOutput(None=SetMaterialAttributes!(
    MaterialAttributes=@n1.MaterialAttributes
    BaseColor=MaterialFunctionCall!(
     Low _ PC Low and Med=@n3
     "High _ XB1, PS4"=@n5
     "Epic _ PC High, XSX, PS5"=@n5
     --------
     "Mobile Low _ PC Perf, Android"=@n3
     Mobile Med _ Switch=@n3
    ) {materialFunction=/Game/Global/MaterialFunctions/TechArt/MF_Fortnite_QualitySwitch_Vector3}.Result
    AmbientOcclusion=NamedRerouteUsage()
   ) {attributeSetTypes [#1, #2]})
  FunctionOutput(None=NamedRerouteUsage()) {outputName=AO Mask, sortPriority=1}
  n6 = FunctionInput() {inputName=TechArtMask, inputType=FunctionInput_Vector4, sortPriority=1}
  n7 = Power(
   Base=OneMinus(None=ChannelMaskParameter(None=@n6) {group=BeaverStage - Helmet Shadow, parameterName=AO - TechArt Mask}.R)
   Exp=ScalarParameter() {defaultValue=1.5, group=BeaverStage - Helmet Shadow, parameterName=AOMask - Power}
  )
  n8 = ScalarParameter() {defaultValue=0.05, group=BeaverStage - Helmet Shadow, parameterName=AOMask - Bias}
  NamedRerouteDeclaration(None=Saturate(None=Add(A=@n7, B=@n8))) {name=AO, nodeColor=<rgba 0 0.5781 1 1>}
 ]}
