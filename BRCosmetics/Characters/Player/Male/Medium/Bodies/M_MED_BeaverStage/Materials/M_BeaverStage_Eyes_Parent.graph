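# Parent material for the eyes of the BeaverStage male player character, built on material attributes for skeletal meshes with Default Lit and Subsurface shading.
# The shared 2023 eye function feeds the eyelash function, then the character-specific MF_BeaverStage function with its TechArtMask fixed at full.
# The result passes through tech-art overrides and ends in the shared character de-rez transition function.
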Material /BRCosmetics/Characters/Player/Male/Medium/Bodies/M_MED_BeaverStage/Materials/M_BeaverStage_Eyes_Parent
graph {
 properties {
  bUseMaterialAttributes=true
  bUsedWithMeshDeformer=true
  bUsedWithSkeletalMesh=true
  pixelDepthOffsetMode=PDOM_Legacy
  shadingModel=MSM_FromMaterialExpression
  usedShadingModels=Default Lit | Subsurface
 }
 nodes [
  n1 = MaterialFunctionCall!(Material=MaterialFunctionCall() {materialFunction=/Game/Global/Materials/Character_Parent_2023/Eye_Parent_2023/MF_eyes_2023}.Result) {materialFunction=/Game/Global/Materials/Character_Parent_2023/Eye_Parent_2023/MF_Eyelash}
  n2 = Constant() {r=1}
  MaterialFunctionCall!(
   Fresnel
   Material=MaterialFunctionCall(In=MaterialFunctionCall!(In=@n1.Result, TechArtMask=@n2) {materialFunction=&/MF_BeaverStage}.Result) {materialFunction=&MF_TechArtOverrides}.Result
  ) {materialFunction=/Game/Packages/Fortress_SharedMaterials/Base_Material_Functions/CharacterDeRez_Transition}
 ]
}
