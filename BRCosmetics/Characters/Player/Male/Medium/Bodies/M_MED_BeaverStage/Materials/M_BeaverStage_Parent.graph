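# Parent character material for the BeaverStage male medium body, a skeletal mesh material using material attributes and a shading model chosen by the graph.
# It runs the shared 2023 character PreFX function, inserts the skin-specific MF_BeaverStage fed by the tech art mask, then applies the shared PostFX function.
# The final emissive color is blended from the default A input by the AO mask from MF_BeaverStage, gating glow with that mask.
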Material /BRCosmetics/Characters/Player/Male/Medium/Bodies/M_MED_BeaverStage/Materials/M_BeaverStage_Parent
graph {
 properties {
  bAutomaticallySetUsageInEditor=false
  bUseMaterialAttributes=true
  bUsedWithMeshDeformer=true
  bUsedWithMorphTargets=true
  bUsedWithSkeletalMesh=true
  pixelDepthOffsetMode=PDOM_Legacy
  shadingModel=MSM_FromMaterialExpression
  usedShadingModels=Default Lit | Subsurface | Clear Coat | Subsurface Profile | Cloth
 }
 nodes [
  n1 = MaterialFunctionCall() {materialFunction=&MF_CharacterParent_2023_PreFX}
  n2 = MaterialFunctionCall!(In=@n1.Result, TechArtMask=AppendVector(A=@n1.Tech Art Mask, B=@n1.Tech Art Mask)) {materialFunction=&/MF_BeaverStage}
  n3 = GetMaterialAttributes(None=MaterialFunctionCall(Material=@n2.Result, M=@n1.M, Pre Skinned Local Position=@n1.M, Distance Blend=@n1.M, Fresnel=@n1.M) {materialFunction=&MF_CharacterParent_2023_PostFX}.Result) {attributeGetTypes [#1, #2]}
  SetMaterialAttributes(MaterialAttributes=@n3.MaterialAttributes, Emissive Color=LinearInterpolate(B=@n3.Emissive Color, Alpha=@n2.AO Mask)) {attributeSetTypes [#1]}
 ]
}
