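# Cloth physics asset for the BeaverStage medium male player body, apparently providing collision volumes for simulated cloth.
# Capsules cover torso, arms, thighs, sword case, backpack attach point, rear skirt bones and outer tail bones; calf bodies have no shapes.
# A large box on root sits around ground level, likely as a floor collider.
# Constraints lock the torso, arms and thighs together, leave pelvis-to-root free, and omit accessory bodies.
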PhysicsAsset /BRCosmetics/Characters/Player/Male/Medium/Bodies/M_MED_BeaverStage/Meshes/M_MED_BeaverStage_PhysicsAsset_Cloth
bodies {
 Base_Sword_case [{
   shapeType=Capsule
   length=57.1664
   radius=4.6074
   shapeName=Base_Sword_case_capsule
   center=<18.286 0.3841 -0.3271>
   rotation=<rot 0 -89.2355 -89.3219>
  }]
 Dyn_skirt_btw_bk_l_02 [{
   shapeType=Capsule
   length=25.4689
   radius=8.1671
   shapeName=Dyn_skirt_btw_bk_l_02_capsule
   center=<4.2137 -7.2138 -3.2858>
   rotation=<rot 90 -4.0291 0>
  }]
 Dyn_skirt_btw_bk_r_02 [{
   shapeType=Capsule
   length=25.4689
   radius=8.1671
   shapeName=Dyn_skirt_btw_bk_r_02_capsule
   center=<4.4865 -4.1042 -0.3738>
   rotation=<rot 90 -3.0712 0>
  }]
 Dyn_skirt_mid_bk_02 [{
   shapeType=Capsule
   length=25.4689
   radius=8.1671
   shapeName=Dyn_skirt_mid_bk_02_capsule
   center=<4.4695 -5.2154 -2.9471>
   rotation=<rot 90 -3.7945 0>
  }]
 Dyn_skirt_side_bk_l_02 [{
   shapeType=Capsule
   length=25.4689
   radius=8.1671
   shapeName=Dyn_skirt_side_bk_l_02_capsule
   center=<9.2656 -4.2828 -2.2346>
   rotation=<rot 83.4585 174.6444 -173.3082>
  }]
 Dyn_skirt_side_bk_r_02 [{
   shapeType=Capsule
   length=25.4689
   radius=8.1671
   shapeName=Dyn_skirt_side_bk_r_02_capsule
   center=<7.7626 -3.374 -0.0265>
   rotation=<rot 65.1357 71.5927 79.0511>
  }]
 Tail_06 [{
   shapeType=Capsule
   length=2.5889
   radius=3
   shapeName=Tail_06_capsule
   center=<-7.3329 2.7356 1.1011>
   rotation=<rot -2.6619 -71.9631 -81.8831>
  }]
 Tail_07 [{
   shapeType=Capsule
   length=4.0194
   radius=3
   shapeName=Tail_07_capsule
   center=<-5.5152 3.5355 0.8281>
   rotation=<rot -2.6619 -71.9631 -81.8831>
  }]
 Tail_08 [{
   shapeType=Capsule
   length=3.6878
   radius=3.5197
   shapeName=Tail_08_capsule
   center=<-3.5759 4.6682 0.5369>
   rotation=<rot -3.549 -65.6036 -82.228>
  }]
 Tail_09 [{
   shapeType=Capsule
   length=7.2043
   radius=3.3045
   shapeName=Tail_09_capsule
   center=<1.7179 6.1545 -0.2579>
   rotation=<rot -2.6619 -71.9631 -81.8831>
  }]
 attach_backpack [{shapeType=Capsule, length=1, radius=20, shapeName=backpack_collision_capsule, center=<0 -41.1598 1.5422>}]
 calf_l []
 calf_r []
 lowerarm_l [{shapeType=Capsule, length=22.9149, radius=7.49, center=<9.8422 1.8249 0>, rotation=<rot 90 0 0>}]
 lowerarm_r [{shapeType=Capsule, length=22.9149, radius=7.49, center=<-9.8422 -1.8249 0>, rotation=<rot 90 180 0>}]
 pelvis [{shapeType=Capsule, length=16.9523, radius=14.0559, center=<0 -2.4832 0>}]
 root [{shapeType=Box, extentX=258.5, extentY=266.5, extentZ=59.5, shapeName=root_box, center=<0 0 -27.1671>, rotation=<rot 0 -90 0>}]
 spine_01 [{shapeType=Capsule, length=7.2569, radius=14.2531, shapeName=spine_01_capsule, center=<6.6223 -1.0891 0>}]
 spine_04 [{shapeType=Capsule, length=13.4914, radius=11.5563, center=<0 -1.3895 0>}]
 spine_05 [{shapeType=Capsule, length=23.5162, radius=14.9612, center=<-4.6566 0.8336 0>}]
 thigh_l [{shapeType=Capsule, length=36.4039, radius=11.381, shapeName=thigh_l_capsule, center=<-18.1103 -0.2367 0>, rotation=<rot 90 0 0>}]
 thigh_r [{shapeType=Capsule, length=36.4039, radius=11.381, shapeName=thigh_r_capsule, center=<18.1103 0.2367 0>, rotation=<rot 90 180 0>}]
 upperarm_l [{shapeType=Capsule, length=24.7594, radius=8.5919, center=<7.9838 -1.4248 -0.3483>, rotation=<rot 90 5.4547 0>}]
 upperarm_r [{shapeType=Capsule, length=24.7594, radius=8.5919, center=<-7.9838 1.4248 0.3483>, rotation=<rot 90 -174.5453 0>}]
}
constraints [
 {
  bone1Name=pelvis
  bone2Name=root
  swing1LimitDegrees=45
  swing1Motion=Free
  swing2LimitDegrees=45
  swing2Motion=Free
  twistLimitDegrees=45
  twistMotion=Free
 }
 {
  bone1Name=spine_01
  bone2Name=pelvis
  swing1LimitDegrees=45
  swing1Motion=Locked
  swing2LimitDegrees=45
  swing2Motion=Locked
  twistLimitDegrees=45
  twistMotion=Locked
 }
 {
  bone1Name=thigh_l
  bone2Name=pelvis
  swing1LimitDegrees=45
  swing1Motion=Locked
  swing2LimitDegrees=45
  swing2Motion=Locked
  twistLimitDegrees=45
  twistMotion=Locked
 }
 {
  bone1Name=thigh_r
  bone2Name=pelvis
  swing1LimitDegrees=45
  swing1Motion=Locked
  swing2LimitDegrees=45
  swing2Motion=Locked
  twistLimitDegrees=45
  twistMotion=Locked
 }
 {
  bone1Name=spine_04
  bone2Name=spine_01
  swing1LimitDegrees=45
  swing1Motion=Locked
  swing2LimitDegrees=45
  swing2Motion=Locked
  twistLimitDegrees=45
  twistMotion=Locked
 }
 {
  bone1Name=spine_05
  bone2Name=spine_04
  swing1LimitDegrees=45
  swing1Motion=Locked
  swing2LimitDegrees=45
  swing2Motion=Locked
  twistLimitDegrees=45
  twistMotion=Locked
 }
 {
  bone1Name=lowerarm_l
  bone2Name=upperarm_l
  swing1LimitDegrees=45
  swing1Motion=Locked
  swing2LimitDegrees=45
  swing2Motion=Locked
  twistLimitDegrees=45
  twistMotion=Locked
 }
 {
  bone1Name=upperarm_r
  bone2Name=spine_05
  swing1LimitDegrees=45
  swing1Motion=Locked
  swing2LimitDegrees=45
  swing2Motion=Locked
  twistLimitDegrees=45
  twistMotion=Locked
 }
 {
  bone1Name=lowerarm_r
  bone2Name=upperarm_r
  swing1LimitDegrees=45
  swing1Motion=Locked
  swing2LimitDegrees=45
  swing2Motion=Locked
  twistLimitDegrees=45
  twistMotion=Locked
 }
 {
  bone1Name=upperarm_l
  bone2Name=spine_05
  swing1LimitDegrees=45
  swing1Motion=Locked
  swing2LimitDegrees=45
  swing2Motion=Locked
  twistLimitDegrees=45
  twistMotion=Locked
 }
]
