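# Physics asset for the BeaverStage medium male player body, NoBackBling variant, defining collision and joints for body, cape, skirt, sword case and tail.
# Capsules cover torso and limbs, with a flat root box below the feet like a floor.
# Four cape chains use tight swing and twist limits, with some free cross-links between them.
# Most body joints are free, tail and elbows are locked, and skirt capsules have no constraints.
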PhysicsAsset /BRCosmetics/Characters/Player/Male/Medium/Bodies/M_MED_BeaverStage/Meshes/M_MED_BeaverStage_PhysicsAsset_NoBackBling
bodies {
 Base_Sword_case [{
   shapeType=Capsule
   length=68.1664
   radius=4.6074
   shapeName=Base_Sword_case_capsule
   center=<14.425 0.3326 -0.2814>
   rotation=<rot 0 -89.2355 -89.3219>
  }]
 Dyn_skirt_bk_l_02 [{
   shapeType=Capsule
   length=32.3029
   radius=7
   shapeName=Dyn_skirt_bk_l_02_capsule
   center=<0.6056 -1.0885 0.1461>
   rotation=<rot 87.4582 -179.9997 -179.9996>
  }]
 Dyn_skirt_bk_r_02 [{
   shapeType=Capsule
   length=32.2958
   radius=7
   shapeName=Dyn_skirt_bk_r_02_capsule
   center=<1.1606 -1.8804 0.5065>
   rotation=<rot 84.916 0.0002 0.0004>
  }]
 Dyn_skirt_btw_bk_l_02 [{
   shapeType=Capsule
   length=33.4363
   radius=7
   shapeName=Dyn_skirt_btw_bk_l_02_capsule
   center=<-0.0601 -1.8498 -0.1223>
   rotation=<rot 87.4581 -180 -179.9998>
  }]
 Dyn_skirt_btw_bk_r_02 [{
   shapeType=Capsule
   length=35.3895
   radius=7
   shapeName=Dyn_skirt_btw_bk_r_02_capsule
   center=<-0.48 -2.0704 -1.6794>
   rotation=<rot 89.5433 0.0001 0.0004>
  }]
 Dyn_skirt_mid_bk_02 [{
   shapeType=Capsule
   length=34.6488
   radius=7
   shapeName=Dyn_skirt_mid_bk_02_capsule
   center=<0.6505 -1.9086 -0.5143>
   rotation=<rot 88.4274 -0.0002 0>
  }]
 Dyn_skirt_side_bk_l_02 [{
   shapeType=Capsule
   length=34.2642
   radius=7
   shapeName=Dyn_skirt_side_bk_l_02_capsule
   center=<3.1754 -0.5695 0.7958>
   rotation=<rot 87.4582 -179.9992 -170.9515>
  }]
 Tail_07 [{
   shapeType=Capsule
   length=7.5
   radius=3.5
   shapeName=Tail_07_capsule
   center=<-5.4337 1.415 -0.9581>
   rotation=<rot 1.7279 -80.1491 -99.8511>
  }]
 Tail_08 [{
   shapeType=Capsule
   length=7.5
   radius=3.5
   shapeName=Tail_08_capsule
   center=<-0.5114 2.9453 -0.0902>
   rotation=<rot 1.7279 -80.1491 -99.8511>
  }]
 Tail_09 [{shapeType=Capsule, length=7.5, radius=3.5, shapeName=Tail_09_capsule, center=<6.2751 3.5114 1.1065>, rotation=<rot 1.7279 -80.149 -99.8511>}]
 calf_l [{shapeType=Capsule, length=36.2963, radius=14.5787, center=<-26.75 0.0987 0>, rotation=<rot 90 -1.4887 0>}]
 calf_r [{shapeType=Capsule, length=36.2963, radius=14.5787, center=<26.75 -0.0987 0>, rotation=<rot 90 178.5113 0>}]
 cape_01_l [{shapeType=Capsule, length=1, radius=1}]
 cape_01_r [{shapeType=Capsule, length=1, radius=1}]
 cape_02_l [{shapeType=Capsule, length=12, radius=9, center=<9 0 10>, rotation=<rot 90 0 0>}]
 cape_02_r [{shapeType=Capsule, length=12, radius=9, center=<9 0 10>, rotation=<rot 90 0 0>}]
 cape_03_l [{shapeType=Capsule, length=12, radius=12, center=<12 0 10>, rotation=<rot 90 0 0>}]
 cape_03_r [{shapeType=Capsule, length=12, radius=12, center=<12 0 10>, rotation=<rot 90 0 0>}]
 cape_04_l [{shapeType=Capsule, length=12, radius=12.723, center=<9 0 13.502>, rotation=<rot 90 0 0>}]
 cape_04_r [{shapeType=Capsule, length=12, radius=12.723, center=<9 0 13.502>, rotation=<rot 90 0 0>}]
 cape_05_l [{shapeType=Capsule, length=12, radius=12.723, center=<9 0 13.502>, rotation=<rot 90 0 0>}]
 cape_05_r [{shapeType=Capsule, length=12, radius=12.723, center=<9 0 13.502>, rotation=<rot 90 0 0>}]
 cape_06_l [{shapeType=Capsule, length=12, radius=12.723, center=<9 0 13.502>, rotation=<rot 90 0 0>}]
 cape_06_r [{shapeType=Capsule, length=12, radius=12.723, center=<9 0 13.502>, rotation=<rot 90 0 0>}]
 cape_07_l [{shapeType=Capsule, length=12, radius=12.723, center=<9 0 13.502>, rotation=<rot 90 0 0>}]
 cape_07_r [{shapeType=Capsule, length=12, radius=12.723, center=<9 0 13.502>, rotation=<rot 90 0 0>}]
 cape_mid_01_l [{shapeType=Capsule, length=12, radius=9, shapeName=cape_mid_01_l_capsule, center=<9.5735 0 11.1454>, rotation=<rot 0 -90 87.9057>}]
 cape_mid_01_l_MSTR [{shapeType=Capsule, length=1, radius=1}]
 cape_mid_01_r [{shapeType=Capsule, length=12, radius=9, center=<9 0 10>, rotation=<rot 90 0 0>}]
 cape_mid_01_r_MSTR [{shapeType=Capsule, length=1, radius=1}]
 cape_mid_02_l [{shapeType=Capsule, length=12, radius=12, center=<12 0 10>, rotation=<rot 90 0 0>}]
 cape_mid_02_r [{shapeType=Capsule, length=12, radius=12, center=<12 0 10>, rotation=<rot 90 0 0>}]
 cape_mid_03_l [{shapeType=Capsule, length=12, radius=12.723, center=<9 0 13.502>, rotation=<rot 90 0 0>}]
 cape_mid_03_r [{shapeType=Capsule, length=12, radius=12.723, center=<9 0 13.502>, rotation=<rot 90 0 0>}]
 cape_mid_04_l [{shapeType=Capsule, length=12, radius=12.723, center=<9 0 13.502>, rotation=<rot 90 0 0>}]
 cape_mid_04_r [{shapeType=Capsule, length=12, radius=12.723, center=<9 0 13.502>, rotation=<rot 90 0 0>}]
 cape_mid_05_l [{shapeType=Capsule, length=12, radius=12.723, center=<9 0 13.502>, rotation=<rot 90 0 0>}]
 cape_mid_05_r [{shapeType=Capsule, length=12, radius=12.723, center=<9 0 13.502>, rotation=<rot 90 0 0>}]
 cape_mid_06_l [{shapeType=Capsule, length=12, radius=12.723, center=<9 0 13.502>, rotation=<rot 90 0 0>}]
 cape_mid_06_r [{shapeType=Capsule, length=12, radius=12.723, center=<9 0 13.502>, rotation=<rot 90 0 0>}]
 clavicle_cape_l [{shapeType=Capsule, length=1, radius=1}]
 clavicle_cape_r [{shapeType=Capsule, length=1, radius=1}]
 clavicle_l [{shapeType=Capsule, length=12.9688, radius=9.6562, center=<12.3638 -1.6165 0>, rotation=<rot 90 0 -6.5611>}]
 clavicle_r [{shapeType=Capsule, length=12.9688, radius=9.6562, center=<-12.3638 1.6165 0>, rotation=<rot 90 0 173.4389>}]
 lowerarm_l [{shapeType=Capsule, length=14.3712, radius=6.1956, center=<9.8422 0 0>, rotation=<rot 90 0 0>}]
 lowerarm_r [{shapeType=Capsule, length=14.3712, radius=6.1956, center=<-9.8422 0 0>, rotation=<rot 90 180 0>}]
 pelvis [{shapeType=Capsule, length=22, radius=18, center=<-7.5688 -0.6735 0>}]
 root [{shapeType=Box, extentX=258.5, extentY=266.5, extentZ=59.5, shapeName=root_box, center=<0 0 -30.3723>, rotation=<rot 0 -90 0>}]
 spine_01 [{shapeType=Capsule, length=4.0312, radius=12, center=<3.5949 1 0>}]
 spine_02 [{shapeType=Capsule, length=3.8125, radius=12, center=<3.5949 0 0>}]
 spine_03 [{shapeType=Capsule, length=39.6562, radius=10, center=<2.5875 -1 0>}]
 spine_04 [{shapeType=Capsule, length=31.4062, radius=12, center=<2.8832 -0.1618 0>}]
 spine_05 [{shapeType=Capsule, length=32.875, radius=12, center=<-0.0029 0.1172 0>}]
 thigh_l [{shapeType=Capsule, length=25.0938, radius=12.2188, center=<-19.7778 -0.1788 0>, rotation=<rot 90 2.6271 0>}]
 thigh_r [{shapeType=Capsule, length=25.0938, radius=12.2188, center=<19.7778 0.1788 0>, rotation=<rot 90 -10.6197 166.7532>}]
 upperarm_l [{shapeType=Capsule, length=20.625, radius=6.9375, center=<11.9672 -1.662 0>, rotation=<rot 90 0 -6.5611>}]
 upperarm_r [{shapeType=Capsule, length=20.625, radius=6.9375, center=<-11.9672 1.662 0>, rotation=<rot 90 0 173.4389>}]
}
constraints [
 {
  bone1Name=pelvis
  bone2Name=root
  swing1LimitDegrees=45
  swing1Motion=Free
  swing2LimitDegrees=45
  swing2Motion=Free
  twistLimitDegrees=45
  twistMotion=Free
 }
 {
  bone1Name=spine_01
  bone2Name=pelvis
  swing1LimitDegrees=45
  swing1Motion=Free
  swing2LimitDegrees=45
  swing2Motion=Free
  twistLimitDegrees=45
  twistMotion=Free
 }
 {
  bone1Name=thigh_l
  bone2Name=pelvis
  swing1LimitDegrees=45
  swing1Motion=Free
  swing2LimitDegrees=45
  swing2Motion=Free
  twistLimitDegrees=45
  twistMotion=Free
 }
 {
  bone1Name=thigh_r
  bone2Name=pelvis
  swing1LimitDegrees=45
  swing1Motion=Free
  swing2LimitDegrees=45
  swing2Motion=Free
  twistLimitDegrees=45
  twistMotion=Free
 }
 {
  bone1Name=spine_02
  bone2Name=spine_01
  swing1LimitDegrees=45
  swing1Motion=Free
  swing2LimitDegrees=45
  swing2Motion=Free
  twistLimitDegrees=45
  twistMotion=Free
 }
 {
  bone1Name=spine_03
  bone2Name=spine_02
  swing1LimitDegrees=45
  swing1Motion=Free
  swing2LimitDegrees=45
  swing2Motion=Free
  twistLimitDegrees=45
  twistMotion=Free
 }
 {
  bone1Name=spine_04
  bone2Name=spine_03
  swing1LimitDegrees=45
  swing1Motion=Free
  swing2LimitDegrees=45
  swing2Motion=Free
  twistLimitDegrees=45
  twistMotion=Free
 }
 {
  bone1Name=spine_05
  bone2Name=spine_04
  swing1LimitDegrees=45
  swing1Motion=Free
  swing2LimitDegrees=45
  swing2Motion=Free
  twistLimitDegrees=45
  twistMotion=Free
 }
 {
  bone1Name=clavicle_r
  bone2Name=spine_05
  swing1LimitDegrees=45
  swing1Motion=Free
  swing2LimitDegrees=45
  swing2Motion=Free
  twistLimitDegrees=45
  twistMotion=Free
 }
 {
  bone1Name=upperarm_l
  bone2Name=spine_05
  swing1LimitDegrees=45
  swing1Motion=Locked
  swing2LimitDegrees=45
  swing2Motion=Locked
  twistLimitDegrees=45
  twistMotion=Locked
 }
 {
  bone1Name=lowerarm_l
  bone2Name=upperarm_l
  swing1LimitDegrees=45
  swing1Motion=Locked
  swing2LimitDegrees=45
  swing2Motion=Locked
  twistLimitDegrees=45
  twistMotion=Locked
 }
 {
  bone1Name=upperarm_r
  bone2Name=clavicle_r
  swing1LimitDegrees=45
  swing1Motion=Free
  swing2LimitDegrees=45
  swing2Motion=Free
  twistLimitDegrees=45
  twistMotion=Free
 }
 {
  bone1Name=lowerarm_r
  bone2Name=upperarm_r
  swing1LimitDegrees=45
  swing1Motion=Locked
  swing2LimitDegrees=45
  swing2Motion=Locked
  twistLimitDegrees=45
  twistMotion=Locked
 }
 {bone1Name=Base_Sword_case, bone2Name=pelvis, swing1LimitDegrees=20, swing2LimitDegrees=20, twistLimitDegrees=5}
 {
  bone1Name=clavicle_l
  bone2Name=spine_05
  swing1LimitDegrees=45
  swing1Motion=Free
  swing2LimitDegrees=45
  swing2Motion=Free
  twistLimitDegrees=45
  twistMotion=Free
 }
 {
  bone1Name=upperarm_l
  bone2Name=clavicle_l
  swing1LimitDegrees=45
  swing1Motion=Free
  swing2LimitDegrees=45
  swing2Motion=Free
  twistLimitDegrees=45
  twistMotion=Free
 }
 {
  bone1Name=cape_mid_01_l_MSTR
  bone2Name=spine_05
  swing1LimitDegrees=45
  swing1Motion=Free
  swing2LimitDegrees=45
  swing2Motion=Free
  twistLimitDegrees=45
  twistMotion=Free
 }
 {
  bone1Name=cape_mid_01_r_MSTR
  bone2Name=spine_05
  swing1LimitDegrees=45
  swing1Motion=Free
  swing2LimitDegrees=45
  swing2Motion=Free
  twistLimitDegrees=45
  twistMotion=Free
 }
 {bone1Name=cape_mid_02_l, bone2Name=cape_mid_01_l, swing1LimitDegrees=5, swing2LimitDegrees=41, twistLimitDegrees=5}
 {bone1Name=cape_mid_03_l, bone2Name=cape_mid_02_l, swing1LimitDegrees=10, swing2LimitDegrees=41, twistLimitDegrees=1}
 {bone1Name=cape_mid_04_l, bone2Name=cape_mid_03_l, swing1LimitDegrees=10, swing2LimitDegrees=42, twistLimitDegrees=1}
 {bone1Name=cape_mid_05_l, bone2Name=cape_mid_04_l, swing1LimitDegrees=10, swing2LimitDegrees=42, twistLimitDegrees=1}
 {
  bone1Name=cape_mid_04_l
  bone2Name=cape_mid_04_r
  swing1LimitDegrees=45
  swing1Motion=Free
  swing2LimitDegrees=45
  swing2Motion=Free
  twistLimitDegrees=45
  twistMotion=Free
 }
 {
  bone1Name=cape_05_l
  bone2Name=cape_mid_04_l
  swing1LimitDegrees=45
  swing1Motion=Free
  swing2LimitDegrees=45
  swing2Motion=Free
  twistLimitDegrees=45
  twistMotion=Free
 }
 {bone1Name=cape_mid_06_l, bone2Name=cape_mid_05_l, swing1LimitDegrees=10, swing2LimitDegrees=45, twistLimitDegrees=1}
 {bone1Name=cape_mid_01_r, bone2Name=cape_mid_01_r_MSTR, swing1LimitDegrees=5, swing2LimitDegrees=40, twistLimitDegrees=5}
 {bone1Name=cape_mid_02_r, bone2Name=cape_mid_01_r, swing1LimitDegrees=5, swing2LimitDegrees=41, twistLimitDegrees=5}
 {bone1Name=cape_mid_03_r, bone2Name=cape_mid_02_r, swing1LimitDegrees=10, swing2LimitDegrees=41, twistLimitDegrees=1}
 {bone1Name=cape_mid_04_r, bone2Name=cape_mid_03_r, swing1LimitDegrees=10, swing2LimitDegrees=42, twistLimitDegrees=1}
 {bone1Name=cape_mid_05_r, bone2Name=cape_mid_04_r, swing1LimitDegrees=10, swing2LimitDegrees=42, twistLimitDegrees=1}
 {
  bone1Name=cape_mid_04_r
  bone2Name=cape_05_r
  swing1LimitDegrees=45
  swing1Motion=Free
  swing2LimitDegrees=45
  swing2Motion=Free
  twistLimitDegrees=45
  twistMotion=Free
 }
 {bone1Name=cape_mid_06_r, bone2Name=cape_mid_05_r, swing1LimitDegrees=10, swing2LimitDegrees=45, twistLimitDegrees=1}
 {bone1Name=cape_01_l, bone2Name=clavicle_cape_l, swing1LimitDegrees=45, swing2LimitDegrees=45, twistLimitDegrees=5}
 {bone1Name=cape_02_l, bone2Name=cape_01_l, swing1LimitDegrees=5, swing2LimitDegrees=40, twistLimitDegrees=5}
 {bone1Name=cape_03_l, bone2Name=cape_02_l, swing1LimitDegrees=5, swing2LimitDegrees=41, twistLimitDegrees=5}
 {bone1Name=cape_04_l, bone2Name=cape_03_l, swing1LimitDegrees=10, swing2LimitDegrees=41, twistLimitDegrees=1}
 {bone1Name=cape_05_l, bone2Name=cape_04_l, swing1LimitDegrees=10, swing2LimitDegrees=42, twistLimitDegrees=1}
 {bone1Name=cape_06_l, bone2Name=cape_05_l, swing1LimitDegrees=10, swing2LimitDegrees=42, twistLimitDegrees=1}
 {bone1Name=cape_07_l, bone2Name=cape_06_l, swing1LimitDegrees=10, swing2LimitDegrees=45, twistLimitDegrees=1}
 {bone1Name=cape_01_r, bone2Name=clavicle_cape_r, swing1LimitDegrees=45, swing2LimitDegrees=45, twistLimitDegrees=5}
 {bone1Name=cape_02_r, bone2Name=cape_01_r, swing1LimitDegrees=5, swing2LimitDegrees=45, twistLimitDegrees=5}
 {bone1Name=cape_03_r, bone2Name=cape_02_r, swing1LimitDegrees=5, swing2LimitDegrees=41, twistLimitDegrees=5}
 {bone1Name=cape_04_r, bone2Name=cape_03_r, swing1LimitDegrees=10, swing2LimitDegrees=41, twistLimitDegrees=1}
 {bone1Name=cape_05_r, bone2Name=cape_04_r, swing1LimitDegrees=10, swing2LimitDegrees=42, twistLimitDegrees=1}
 {bone1Name=cape_06_r, bone2Name=cape_05_r, swing1LimitDegrees=10, swing2LimitDegrees=42, twistLimitDegrees=1}
 {bone1Name=cape_07_r, bone2Name=cape_06_r, swing1LimitDegrees=10, swing2LimitDegrees=45, twistLimitDegrees=1}
 {
  bone1Name=calf_l
  bone2Name=thigh_l
  swing1LimitDegrees=45
  swing1Motion=Free
  swing2LimitDegrees=45
  swing2Motion=Free
  twistLimitDegrees=45
  twistMotion=Free
 }
 {
  bone1Name=calf_r
  bone2Name=thigh_r
  swing1LimitDegrees=45
  swing1Motion=Free
  swing2LimitDegrees=45
  swing2Motion=Free
  twistLimitDegrees=45
  twistMotion=Free
 }
 {
  bone1Name=Tail_07
  bone2Name=spine_05
  swing1LimitDegrees=45
  swing1Motion=Locked
  swing2LimitDegrees=45
  swing2Motion=Locked
  twistLimitDegrees=45
  twistMotion=Locked
 }
 {
  bone1Name=Tail_08
  bone2Name=Tail_07
  swing1LimitDegrees=45
  swing1Motion=Locked
  swing2LimitDegrees=45
  swing2Motion=Locked
  twistLimitDegrees=45
  twistMotion=Locked
 }
 {
  bone1Name=Tail_09
  bone2Name=Tail_08
  swing1LimitDegrees=45
  swing1Motion=Locked
  swing2LimitDegrees=45
  swing2Motion=Locked
  twistLimitDegrees=45
  twistMotion=Locked
 }
 {bone1Name=cape_mid_01_l, bone2Name=cape_mid_01_l_MSTR, swing1LimitDegrees=5, swing2LimitDegrees=40, twistLimitDegrees=5}
 {
  bone1Name=cape_07_l
  bone2Name=cape_mid_06_l
  swing1LimitDegrees=45
  swing1Motion=Free
  swing2LimitDegrees=45
  swing2Motion=Free
  twistLimitDegrees=45
  twistMotion=Free
 }
]
