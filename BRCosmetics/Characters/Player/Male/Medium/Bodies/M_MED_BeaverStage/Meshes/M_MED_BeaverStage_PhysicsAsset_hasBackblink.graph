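# Physics asset for the BeaverStage medium male player body, in a variant named for back bling.
# Capsules cover body, skirt, sword case and cape, with a large box below the root.
# Body constraints are locked except free pelvis-to-root; skirt and sword case bodies have none.
# Four limited cape chains hang from clavicle and backpack bones, cross-linked midway by free constraints, and cape_mid_03_l carries extra tail-named capsules.
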PhysicsAsset /BRCosmetics/Characters/Player/Male/Medium/Bodies/M_MED_BeaverStage/Meshes/M_MED_BeaverStage_PhysicsAsset_hasBackblink
bodies {
 Base_Sword_case [{
   shapeType=Capsule
   length=68.1664
   radius=4.6074
   shapeName=Base_Sword_case_capsule
   center=<14.425 0.3326 -0.2814>
   rotation=<rot 0 -89.2355 -89.3219>
  }]
 Base_skirt_bk_l [{shapeType=Capsule, length=13.005, radius=0.505, shapeName=Base_skirt_bk_l_capsule, rotation=<rot 0 -90 0>}]
 Base_skirt_bk_r [{shapeType=Capsule, length=13.005, radius=0.505, shapeName=Base_skirt_bk_r_capsule, rotation=<rot 0 -90 0>}]
 Base_skirt_btw_bk_l [{shapeType=Capsule, length=13.005, radius=0.505, shapeName=Base_skirt_btw_bk_l_capsule, rotation=<rot 0 -90 0>}]
 Base_skirt_btw_bk_r [{shapeType=Capsule, length=13.005, radius=0.505, shapeName=Base_skirt_btw_bk_r_capsule, rotation=<rot 0 -90 0>}]
 Base_skirt_mid_bk [{shapeType=Capsule, length=13.005, radius=0.505, shapeName=Base_skirt_mid_bk_capsule, rotation=<rot 0 -90 0>}]
 Base_skirt_side_bk_l [{shapeType=Capsule, length=13.005, radius=0.505, shapeName=Base_skirt_side_bk_l_capsule, rotation=<rot 0 -90 0>}]
 Base_skirt_side_bk_r [{shapeType=Capsule, length=13.005, radius=0.505, shapeName=Base_skirt_side_bk_r_capsule, rotation=<rot 0 -90 0>}]
 Base_skirt_side_l [{
   shapeType=Capsule
   length=19.6664
   radius=4.6052
   shapeName=Base_skirt_side_l_capsule
   center=<18.5177 3.758 3.6893>
   rotation=<rot 0 90.8151 -69.7026>
  }]
 Base_skirt_side_r [{shapeType=Capsule, length=13.005, radius=0.505, shapeName=Base_skirt_side_r_capsule, rotation=<rot 0 -90 0>}]
 Dyn_skirt_bk_l_01 [{
   shapeType=Capsule
   length=24.1408
   radius=3.1321
   shapeName=Dyn_skirt_bk_l_01_capsule
   center=<13.3914 -0.3572 0.3763>
   rotation=<rot 0 88.4249 -89.1859>
  }]
 Dyn_skirt_bk_l_02 [{
   shapeType=Capsule
   length=19.3029
   radius=3.1541
   shapeName=Dyn_skirt_bk_l_02_capsule
   center=<11.1033 -1.0885 0.6121>
   rotation=<rot 0 83.4578 -88.9246>
  }]
 Dyn_skirt_bk_r_01 [{
   shapeType=Capsule
   length=22.9784
   radius=3.1823
   shapeName=Dyn_skirt_bk_r_01_capsule
   center=<14.5755 -0.5001 -0.1519>
   rotation=<rot 0 -94.4121 -89.6521>
  }]
 Dyn_skirt_bk_r_02 [{
   shapeType=Capsule
   length=19.2958
   radius=3.103
   shapeName=Dyn_skirt_bk_r_02_capsule
   center=<11.0981 -1.8804 -0.3776>
   rotation=<rot 0 -98.2834 -88.2085>
  }]
 Dyn_skirt_btw_bk_l_01 [{
   shapeType=Capsule
   length=26.2433
   radius=3.8444
   shapeName=Dyn_skirt_btw_bk_l_01_capsule
   center=<13.5291 -1.4059 0.7592>
   rotation=<rot 0 -92.0006 -89.3042>
  }]
 Dyn_skirt_btw_bk_l_02 [{
   shapeType=Capsule
   length=20.4363
   radius=3.7572
   shapeName=Dyn_skirt_btw_bk_l_02_capsule
   center=<10.0442 -1.8499 0.3262>
   rotation=<rot 0 -103.4287 -87.7868>
  }]
 Dyn_skirt_btw_bk_r_01 [{
   shapeType=Capsule
   length=25.0775
   radius=3.8303
   shapeName=Dyn_skirt_btw_bk_r_01_capsule
   center=<14.4077 -1.7865 -0.775>
   rotation=<rot 0 84.0946 -89.1234>
  }]
 Dyn_skirt_btw_bk_r_02 [{
   shapeType=Capsule
   length=22.3895
   radius=3.5347
   shapeName=Dyn_skirt_btw_bk_r_02_capsule
   center=<7.7746 -2.0704 -0.3738>
   rotation=<rot 0 78.1063 -89.409>
  }]
 Dyn_skirt_mid_bk_01 [{
   shapeType=Capsule
   length=23.963
   radius=3.4731
   shapeName=Dyn_skirt_mid_bk_01_capsule
   center=<11.4133 -0.6181 0.0459>
   rotation=<rot 0 -91.1616 -89.5053>
  }]
 Dyn_skirt_mid_bk_02 [{
   shapeType=Capsule
   length=21.6488
   radius=3.4732
   shapeName=Dyn_skirt_mid_bk_02_capsule
   center=<11.4469 -1.9086 -0.035>
   rotation=<rot 0 82.145 -89.7841>
  }]
 Dyn_skirt_side_bk_l_01 [{
   shapeType=Capsule
   length=27.3291
   radius=4.3577
   shapeName=Dyn_skirt_side_bk_l_01_capsule
   center=<16.5298 -1.832 0.8361>
   rotation=<rot 0 88.2661 -87.1328>
  }]
 Dyn_skirt_side_bk_l_02 [{
   shapeType=Capsule
   length=21.2642
   radius=4.331
   shapeName=Dyn_skirt_side_bk_l_02_capsule
   center=<14.5576 -2.4122 5.3287>
   rotation=<rot 0 80.7406 -88.3724>
  }]
 Dyn_skirt_side_bk_r_01 [{
   shapeType=Capsule
   length=27.3457
   radius=4.2518
   shapeName=Dyn_skirt_side_bk_r_01_capsule
   center=<16.57 -1.6579 -1.2035>
   rotation=<rot 0 -89.1302 -84.6577>
  }]
 Dyn_skirt_side_bk_r_02 [{
   shapeType=Capsule
   length=21.2778
   radius=4.2669
   shapeName=Dyn_skirt_side_bk_r_02_capsule
   center=<14.5873 -1.6892 -2.8133>
   rotation=<rot 0 -98.2005 -77.3251>
  }]
 Dyn_skirt_side_l_01 [{shapeType=Capsule, length=13.005, radius=0.505, shapeName=Dyn_skirt_side_l_01_capsule, rotation=<rot 0 -90 0>}]
 Dyn_skirt_side_l_02 [{shapeType=Capsule, length=13.005, radius=0.505, shapeName=Dyn_skirt_side_l_02_capsule, rotation=<rot 0 -90 0>}]
 Dyn_skirt_side_r_01 [{
   shapeType=Capsule
   length=25.1874
   radius=6.8147
   shapeName=Dyn_skirt_side_r_01_capsule
   center=<13.2442 -1.2993 0.7001>
   rotation=<rot 0 83.3012 -89.9734>
  }]
 Dyn_skirt_side_r_02 [{
   shapeType=Capsule
   length=20.6386
   radius=7.648
   shapeName=Dyn_skirt_side_r_02_capsule
   center=<11.906 -2.5982 0.8862>
   rotation=<rot 0 76.8405 -76.7168>
  }]
 backpack_root [{shapeType=Capsule, length=1, radius=1}]
 calf_l [{shapeType=Capsule, length=25.2372, radius=10.7837, center=<-17.038 -1.9066 -0.3219>, rotation=<rot 90 0 0>}]
 calf_r [{shapeType=Capsule, length=24.2372, radius=11.7837, center=<17.038 2.1081 0.3219>, rotation=<rot 90 180 0>}]
 cape_01_l [{shapeType=Capsule, length=1, radius=1}]
 cape_01_r [{shapeType=Capsule, length=1, radius=1}]
 cape_02_l [{shapeType=Capsule, length=12, radius=9, center=<9 0 10>, rotation=<rot 90 0 0>}]
 cape_02_r [{shapeType=Capsule, length=12, radius=9, center=<9 0 10>, rotation=<rot 90 0 0>}]
 cape_03_l [{shapeType=Capsule, length=12, radius=12, center=<12 0 10>, rotation=<rot 90 0 0>}]
 cape_03_r [{shapeType=Capsule, length=12, radius=12, center=<12 0 10>, rotation=<rot 90 0 0>}]
 cape_04_l [{shapeType=Capsule, length=12, radius=12.723, center=<9 0 13.502>, rotation=<rot 90 0 0>}]
 cape_04_r [{shapeType=Capsule, length=12, radius=12.723, center=<9 0 13.502>, rotation=<rot 90 0 0>}]
 cape_05_l [{shapeType=Capsule, length=12, radius=12.723, center=<9 0 13.502>, rotation=<rot 90 0 0>}]
 cape_05_r [{shapeType=Capsule, length=12, radius=12.723, center=<9 0 13.502>, rotation=<rot 90 0 0>}]
 cape_06_l [{shapeType=Capsule, length=12, radius=12.723, center=<9 0 13.502>, rotation=<rot 90 0 0>}]
 cape_06_r [{shapeType=Capsule, length=12, radius=12.723, center=<9 0 13.502>, rotation=<rot 90 0 0>}]
 cape_07_l [{shapeType=Capsule, length=12, radius=12.723, center=<9 0 13.502>, rotation=<rot 90 0 0>}]
 cape_07_r [{shapeType=Capsule, length=12, radius=12.723, center=<9 0 13.502>, rotation=<rot 90 0 0>}]
 cape_mid_01_l_MSTR [{shapeType=Capsule, length=1, radius=1}]
 cape_mid_01_r [{shapeType=Capsule, length=12, radius=9, center=<9 0 10>, rotation=<rot 90 0 0>}]
 cape_mid_01_r_MSTR [{shapeType=Capsule, length=1, radius=1}]
 cape_mid_02_l [{shapeType=Capsule, length=12, radius=12, center=<12 0 10>, rotation=<rot 90 0 0>}]
 cape_mid_02_r [{shapeType=Capsule, length=12, radius=12, center=<12 0 10>, rotation=<rot 90 0 0>}]
 cape_mid_03_l [
  {shapeType=Capsule, length=12, radius=12.723, center=<9 0 13.502>, rotation=<rot 90 0 0>}
  {
   shapeType=Capsule
   length=2.5889
   radius=3
   shapeName=Tail_06_capsule
   center=<-7.3329 2.7357 1.1011>
   rotation=<rot -2.6619 -71.9631 -81.8831>
  }
  {
   shapeType=Capsule
   length=4.0194
   radius=3
   shapeName=Tail_07_capsule
   center=<-5.5152 3.5355 0.8281>
   rotation=<rot -2.6619 -71.9631 -81.8831>
  }
  {
   shapeType=Capsule
   length=3.6878
   radius=3.5197
   shapeName=Tail_08_capsule
   center=<-3.5759 4.6682 0.5369>
   rotation=<rot -3.549 -65.6036 -82.228>
  }
  {
   shapeType=Capsule
   length=7.2043
   radius=3.3045
   shapeName=Tail_09_capsule
   center=<1.7179 6.1545 -0.2579>
   rotation=<rot -2.6619 -71.9631 -81.8831>
  }
 ]
 cape_mid_03_r [{shapeType=Capsule, length=12, radius=12.723, center=<9 0 13.502>, rotation=<rot 90 0 0>}]
 cape_mid_04_l [{shapeType=Capsule, length=12, radius=12.723, center=<9 0 13.502>, rotation=<rot 90 0 0>}]
 cape_mid_04_r [{shapeType=Capsule, length=12, radius=12.723, center=<9 0 13.502>, rotation=<rot 90 0 0>}]
 cape_mid_05_l [{shapeType=Capsule, length=12, radius=12.723, center=<9 0 13.502>, rotation=<rot 90 0 0>}]
 cape_mid_05_r [{shapeType=Capsule, length=12, radius=12.723, center=<9 0 13.502>, rotation=<rot 90 0 0>}]
 cape_mid_06_l [{shapeType=Capsule, length=12, radius=12.723, center=<9 0 13.502>, rotation=<rot 90 0 0>}]
 cape_mid_06_r [{shapeType=Capsule, length=12, radius=12.723, center=<9 0 13.502>, rotation=<rot 90 0 0>}]
 clavicle_cape_l [{shapeType=Capsule, length=1, radius=1}]
 clavicle_cape_r [{shapeType=Capsule, length=1, radius=1}]
 clavicle_l [{shapeType=Capsule, length=7.0634, radius=6.5229, center=<7 -4 0>, rotation=<rot 90 0 0>}]
 clavicle_r [{shapeType=Capsule, length=7.0634, radius=6.5229, center=<-7 4 0>, rotation=<rot 90 180 0>}]
 foot_l [{shapeType=Capsule, length=13.9693, radius=8.9346, center=<-3.004 6.5342 -0.325>, rotation=<rot 0 0 90>}]
 foot_r [{shapeType=Capsule, length=13.9693, radius=8.9346, center=<3.004 -6.5342 0.325>, rotation=<rot 0 0 -90>}]
 hand_l [{shapeType=Capsule, length=9.1708, radius=6.2527, center=<7.7673 -2.6609 0.26>, rotation=<rot 0 81.1706 -79.3237>}]
 hand_r [{shapeType=Capsule, length=8.9257, radius=6.6213, center=<-7.7545 2.6833 -0.6535>, rotation=<rot 0 82.0056 -74.5376>}]
 lowerarm_l [{shapeType=Capsule, length=14.3712, radius=6.1956, center=<9.8422 0 0>, rotation=<rot 90 0 0>}]
 lowerarm_r [{shapeType=Capsule, length=14.3712, radius=6.1956, center=<-9.8422 0 0>, rotation=<rot 90 180 0>}]
 pelvis [{shapeType=Capsule, length=18.5, radius=12.5082, center=<0 -7.2383 0>}]
 root [{shapeType=Box, extentX=258.5, extentY=266.5, extentZ=59.5, shapeName=root_box, center=<0 0 -28.8876>, rotation=<rot 0 -90 0>}]
 spine_01 [{shapeType=Capsule, length=8.005, radius=13.505, shapeName=spine_01_capsule, center=<0 -1.006 0>}]
 spine_02 [{shapeType=Capsule, length=19.3921, radius=10.8899, center=<0 -2.1939 0>}]
 spine_03 [{shapeType=Capsule, length=10.3957, radius=11.4522, center=<0 -1.3308 0>}]
 spine_04 [{shapeType=Capsule, length=12.7863, radius=9.6562, center=<0 -2.4745 0>}]
 spine_05 [{shapeType=Capsule, length=32.875, radius=12, center=<-0.0029 0.1172 0>}]
 thigh_l [{shapeType=Capsule, length=19.3867, radius=11.2219, shapeName=thigh_l_capsule, center=<-18.1103 -3.7326 0>, rotation=<rot 90 0 0>}]
 thigh_r [{shapeType=Capsule, length=19.3867, radius=11.2219, shapeName=thigh_r_capsule, center=<18.1103 3.7326 0>, rotation=<rot 90 180 0>}]
 upperarm_l [{shapeType=Capsule, length=15.4485, radius=5.6809, center=<13.0051 -1 0>, rotation=<rot 90 0 0>}]
 upperarm_r [{shapeType=Capsule, length=15.4485, radius=5.6809, center=<-13.0051 1.8622 0>, rotation=<rot 90 180 0>}]
}
constraints [
 {
  bone1Name=pelvis
  bone2Name=root
  swing1LimitDegrees=45
  swing1Motion=Free
  swing2LimitDegrees=45
  swing2Motion=Free
  twistLimitDegrees=45
  twistMotion=Free
 }
 {
  bone1Name=spine_01
  bone2Name=pelvis
  swing1LimitDegrees=45
  swing1Motion=Locked
  swing2LimitDegrees=45
  swing2Motion=Locked
  twistLimitDegrees=45
  twistMotion=Locked
 }
 {
  bone1Name=thigh_l
  bone2Name=pelvis
  swing1LimitDegrees=45
  swing1Motion=Locked
  swing2LimitDegrees=45
  swing2Motion=Locked
  twistLimitDegrees=45
  twistMotion=Locked
 }
 {
  bone1Name=thigh_r
  bone2Name=pelvis
  swing1LimitDegrees=45
  swing1Motion=Locked
  swing2LimitDegrees=45
  swing2Motion=Locked
  twistLimitDegrees=45
  twistMotion=Locked
 }
 {
  bone1Name=spine_02
  bone2Name=spine_01
  swing1LimitDegrees=45
  swing1Motion=Locked
  swing2LimitDegrees=45
  swing2Motion=Locked
  twistLimitDegrees=45
  twistMotion=Locked
 }
 {
  bone1Name=spine_03
  bone2Name=spine_02
  swing1LimitDegrees=45
  swing1Motion=Locked
  swing2LimitDegrees=45
  swing2Motion=Locked
  twistLimitDegrees=45
  twistMotion=Locked
 }
 {
  bone1Name=spine_04
  bone2Name=spine_03
  swing1LimitDegrees=45
  swing1Motion=Locked
  swing2LimitDegrees=45
  swing2Motion=Locked
  twistLimitDegrees=45
  twistMotion=Locked
 }
 {
  bone1Name=spine_05
  bone2Name=spine_04
  swing1LimitDegrees=45
  swing1Motion=Locked
  swing2LimitDegrees=45
  swing2Motion=Locked
  twistLimitDegrees=45
  twistMotion=Locked
 }
 {
  bone1Name=clavicle_l
  bone2Name=spine_05
  swing1LimitDegrees=45
  swing1Motion=Locked
  swing2LimitDegrees=45
  swing2Motion=Locked
  twistLimitDegrees=45
  twistMotion=Locked
 }
 {
  bone1Name=clavicle_r
  bone2Name=spine_05
  swing1LimitDegrees=45
  swing1Motion=Locked
  swing2LimitDegrees=45
  swing2Motion=Locked
  twistLimitDegrees=45
  twistMotion=Locked
 }
 {
  bone1Name=upperarm_l
  bone2Name=clavicle_l
  swing1LimitDegrees=45
  swing1Motion=Locked
  swing2LimitDegrees=45
  swing2Motion=Locked
  twistLimitDegrees=45
  twistMotion=Locked
 }
 {
  bone1Name=lowerarm_l
  bone2Name=upperarm_l
  swing1LimitDegrees=45
  swing1Motion=Locked
  swing2LimitDegrees=45
  swing2Motion=Locked
  twistLimitDegrees=45
  twistMotion=Locked
 }
 {
  bone1Name=hand_l
  bone2Name=lowerarm_l
  swing1LimitDegrees=45
  swing1Motion=Locked
  swing2LimitDegrees=45
  swing2Motion=Locked
  twistLimitDegrees=45
  twistMotion=Locked
 }
 {
  bone1Name=upperarm_r
  bone2Name=clavicle_r
  swing1LimitDegrees=45
  swing1Motion=Locked
  swing2LimitDegrees=45
  swing2Motion=Locked
  twistLimitDegrees=45
  twistMotion=Locked
 }
 {
  bone1Name=lowerarm_r
  bone2Name=upperarm_r
  swing1LimitDegrees=45
  swing1Motion=Locked
  swing2LimitDegrees=45
  swing2Motion=Locked
  twistLimitDegrees=45
  twistMotion=Locked
 }
 {
  bone1Name=hand_r
  bone2Name=lowerarm_r
  swing1LimitDegrees=45
  swing1Motion=Locked
  swing2LimitDegrees=45
  swing2Motion=Locked
  twistLimitDegrees=45
  twistMotion=Locked
 }
 {
  bone1Name=calf_l
  bone2Name=thigh_l
  swing1LimitDegrees=45
  swing1Motion=Locked
  swing2LimitDegrees=45
  swing2Motion=Locked
  twistLimitDegrees=45
  twistMotion=Locked
 }
 {
  bone1Name=foot_l
  bone2Name=calf_l
  swing1LimitDegrees=45
  swing1Motion=Locked
  swing2LimitDegrees=45
  swing2Motion=Locked
  twistLimitDegrees=45
  twistMotion=Locked
 }
 {
  bone1Name=calf_r
  bone2Name=thigh_r
  swing1LimitDegrees=45
  swing1Motion=Locked
  swing2LimitDegrees=45
  swing2Motion=Locked
  twistLimitDegrees=45
  twistMotion=Locked
 }
 {
  bone1Name=foot_r
  bone2Name=calf_r
  swing1LimitDegrees=45
  swing1Motion=Locked
  swing2LimitDegrees=45
  swing2Motion=Locked
  twistLimitDegrees=45
  twistMotion=Locked
 }
 {bone1Name=cape_01_l, bone2Name=clavicle_cape_l, swing1LimitDegrees=45, swing2LimitDegrees=45, twistLimitDegrees=5}
 {bone1Name=cape_02_l, bone2Name=cape_01_l, swing1LimitDegrees=5, swing2LimitDegrees=45, twistLimitDegrees=5}
 {bone1Name=cape_03_l, bone2Name=cape_02_l, swing1LimitDegrees=5, swing2LimitDegrees=45, twistLimitDegrees=5}
 {bone1Name=cape_04_l, bone2Name=cape_03_l, swing1LimitDegrees=10, swing2LimitDegrees=45, twistLimitDegrees=1}
 {bone1Name=cape_05_l, bone2Name=cape_04_l, swing1LimitDegrees=10, swing2LimitDegrees=45, twistLimitDegrees=1}
 {
  bone1Name=cape_05_l
  bone2Name=cape_mid_04_l
  swing1LimitDegrees=45
  swing1Motion=Free
  swing2LimitDegrees=45
  swing2Motion=Free
  twistLimitDegrees=45
  twistMotion=Free
 }
 {bone1Name=cape_06_l, bone2Name=cape_05_l, swing1LimitDegrees=10, swing2LimitDegrees=45, twistLimitDegrees=1}
 {bone1Name=cape_07_l, bone2Name=cape_06_l, swing1LimitDegrees=10, swing2LimitDegrees=45, twistLimitDegrees=1}
 {
  bone1Name=cape_mid_01_r_MSTR
  bone2Name=backpack_root
  swing1LimitDegrees=45
  swing1Motion=Free
  swing2LimitDegrees=45
  swing2Motion=Free
  twistLimitDegrees=45
  twistMotion=Free
 }
 {bone1Name=cape_mid_01_r, bone2Name=cape_mid_01_r_MSTR, swing1LimitDegrees=5, swing2LimitDegrees=45, twistLimitDegrees=5}
 {bone1Name=cape_mid_02_r, bone2Name=cape_mid_01_r, swing1LimitDegrees=5, swing2LimitDegrees=45, twistLimitDegrees=5}
 {bone1Name=cape_mid_03_r, bone2Name=cape_mid_02_r, swing1LimitDegrees=10, swing2LimitDegrees=45, twistLimitDegrees=1}
 {bone1Name=cape_mid_04_r, bone2Name=cape_mid_03_r, swing1LimitDegrees=10, swing2LimitDegrees=45, twistLimitDegrees=1}
 {
  bone1Name=cape_mid_04_l
  bone2Name=cape_mid_04_r
  swing1LimitDegrees=45
  swing1Motion=Free
  swing2LimitDegrees=45
  swing2Motion=Free
  twistLimitDegrees=45
  twistMotion=Free
 }
 {bone1Name=cape_mid_05_r, bone2Name=cape_mid_04_r, swing1LimitDegrees=10, swing2LimitDegrees=45, twistLimitDegrees=1}
 {
  bone1Name=cape_mid_04_r
  bone2Name=cape_05_r
  swing1LimitDegrees=45
  swing1Motion=Free
  swing2LimitDegrees=45
  swing2Motion=Free
  twistLimitDegrees=45
  twistMotion=Free
 }
 {bone1Name=cape_mid_06_r, bone2Name=cape_mid_05_r, swing1LimitDegrees=10, swing2LimitDegrees=45, twistLimitDegrees=1}
 {
  bone1Name=backpack_root
  bone2Name=spine_05
  swing1LimitDegrees=45
  swing1Motion=Free
  swing2LimitDegrees=45
  swing2Motion=Free
  twistLimitDegrees=45
  twistMotion=Free
 }
 {
  bone1Name=cape_mid_01_l_MSTR
  bone2Name=backpack_root
  swing1LimitDegrees=45
  swing1Motion=Free
  swing2LimitDegrees=45
  swing2Motion=Free
  twistLimitDegrees=45
  twistMotion=Free
 }
 {bone1Name=cape_mid_02_l, bone2Name=cape_mid_01_l_MSTR, swing1LimitDegrees=5, swing2LimitDegrees=45, twistLimitDegrees=5}
 {bone1Name=cape_mid_03_l, bone2Name=cape_mid_02_l, swing1LimitDegrees=10, swing2LimitDegrees=45, twistLimitDegrees=1}
 {bone1Name=cape_mid_04_l, bone2Name=cape_mid_03_l, swing1LimitDegrees=10, swing2LimitDegrees=45, twistLimitDegrees=1}
 {bone1Name=cape_mid_05_l, bone2Name=cape_mid_04_l, swing1LimitDegrees=10, swing2LimitDegrees=45, twistLimitDegrees=1}
 {bone1Name=cape_mid_06_l, bone2Name=cape_mid_05_l, swing1LimitDegrees=10, swing2LimitDegrees=45, twistLimitDegrees=1}
 {bone1Name=cape_01_r, bone2Name=clavicle_cape_r, swing1LimitDegrees=45, swing2LimitDegrees=45, twistLimitDegrees=5}
 {bone1Name=cape_02_r, bone2Name=cape_01_r, swing1LimitDegrees=5, swing2LimitDegrees=45, twistLimitDegrees=5}
 {bone1Name=cape_03_r, bone2Name=cape_02_r, swing1LimitDegrees=5, swing2LimitDegrees=45, twistLimitDegrees=5}
 {bone1Name=cape_04_r, bone2Name=cape_03_r, swing1LimitDegrees=10, swing2LimitDegrees=45, twistLimitDegrees=1}
 {bone1Name=cape_05_r, bone2Name=cape_04_r, swing1LimitDegrees=10, swing2LimitDegrees=45, twistLimitDegrees=1}
 {bone1Name=cape_06_r, bone2Name=cape_05_r, swing1LimitDegrees=10, swing2LimitDegrees=45, twistLimitDegrees=1}
 {bone1Name=cape_07_r, bone2Name=cape_06_r, swing1LimitDegrees=10, swing2LimitDegrees=45, twistLimitDegrees=1}
]
