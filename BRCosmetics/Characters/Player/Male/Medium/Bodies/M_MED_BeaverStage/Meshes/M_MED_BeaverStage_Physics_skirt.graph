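# Physics asset for the front skirt of the male medium BeaverStage cosmetic body, with capsules on pelvis, thighs and skirt bones.
# Four tiny base capsules attach to the pelvis with free rotation, each leading into a chain of one or two Dyn capsules.
# Left and middle chains lock twist and mostly limit swing to one axis.
# The right-between chain has one fully free segment, and no front-right chain exists.
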PhysicsAsset /BRCosmetics/Characters/Player/Male/Medium/Bodies/M_MED_BeaverStage/Meshes/M_MED_BeaverStage_Physics_skirt
bodies {
 Base_skirt_btw_fr_l [{shapeType=Capsule, length=0.505, radius=0.505, shapeName=Base_skirt_btw_fr_l_capsule, rotation=<rot 0 -90 0>}]
 Base_skirt_btw_fr_r [{shapeType=Capsule, length=0.505, radius=0.505, shapeName=Base_skirt_btw_fr_r_capsule, rotation=<rot 0 -90 0>}]
 Base_skirt_fr_l [{shapeType=Capsule, length=0.505, radius=0.505, shapeName=Base_skirt_fr_l_capsule, rotation=<rot 0 -90 0>}]
 Base_skirt_mid_fr [{shapeType=Capsule, length=0.505, radius=0.505, shapeName=Base_skirt_mid_fr_capsule, rotation=<rot 0 -90 0>}]
 Dyn_skirt_btw_fr_l_01 [{
   shapeType=Capsule
   length=18.4477
   radius=3.1475
   shapeName=Dyn_skirt_btw_fr_l_01_capsule
   center=<12.0308 2.9985 -0.8934>
   rotation=<rot 77.5185 95.1066 95.2025>
  }]
 Dyn_skirt_btw_fr_l_02 [{
   shapeType=Capsule
   length=8.3525
   radius=3.5861
   shapeName=Dyn_skirt_btw_fr_l_02_capsule
   center=<11.7392 2.8928 -0.2578>
   rotation=<rot 77.5185 95.1066 95.2025>
  }]
 Dyn_skirt_btw_fr_r_01 [{
   shapeType=Capsule
   length=24.6457
   radius=3.0002
   shapeName=Dyn_skirt_btw_fr_r_01_capsule
   center=<22.561 -0.7641 -0.882>
   rotation=<rot 0 -90.8671 -87.1202>
  }]
 Dyn_skirt_fr_l_01 [{
   shapeType=Capsule
   length=17.2644
   radius=3.5355
   shapeName=Dyn_skirt_fr_l_01_capsule
   center=<14.0123 6.8 -2.4496>
   rotation=<rot 61.7913 91.8624 92.1131>
  }]
 Dyn_skirt_fr_l_02 [{
   shapeType=Capsule
   length=7.6301
   radius=4.1533
   shapeName=Dyn_skirt_fr_l_02_capsule
   center=<13.9657 4.6081 0.2271>
   rotation=<rot 61.7913 91.8624 92.1131>
  }]
 Dyn_skirt_mid_fr_01 [{
   shapeType=Capsule
   length=19.1718
   radius=3.5553
   shapeName=Dyn_skirt_mid_fr_01_capsule
   center=<10.4718 4.5082 -1.3318>
   rotation=<rot 89.6661 -180 -180>
  }]
 Dyn_skirt_mid_fr_02 [{
   shapeType=Capsule
   length=7.9365
   radius=3.2995
   shapeName=Dyn_skirt_mid_fr_02_capsule
   center=<11.993 3.1617 -0.6652>
   rotation=<rot 89.6661 -180 -180>
  }]
 pelvis [{shapeType=Capsule, length=4.3717, radius=13.4561, center=<-6.1959 -2.6564 -0.1039>, rotation=<rot -0.0016 -91.714 -1.2037>}]
 thigh_l [{shapeType=Capsule, length=19.3007, radius=8.0295, center=<-19.6406 0.3819 -0.0877>, rotation=<rot 0 -87.931 -84.552>}]
 thigh_r [{shapeType=Capsule, length=19.3007, radius=8.0295, center=<19.6406 -0.3819 0.0877>, rotation=<rot 0 -87.931 95.448>}]
}
constraints [
 {bone1Name=thigh_l, bone2Name=pelvis, swing1LimitDegrees=45, swing2LimitDegrees=45, twistLimitDegrees=45}
 {bone1Name=thigh_r, bone2Name=pelvis, swing1LimitDegrees=45, swing2LimitDegrees=45, twistLimitDegrees=45}
 {
  bone1Name=Base_skirt_fr_l
  bone2Name=pelvis
  swing1LimitDegrees=30
  swing1Motion=Free
  swing2LimitDegrees=30
  swing2Motion=Free
  twistLimitDegrees=10
  twistMotion=Free
 }
 {
  bone1Name=Dyn_skirt_fr_l_01
  bone2Name=Base_skirt_fr_l
  swing1LimitDegrees=30
  swing2LimitDegrees=2
  swing2Motion=Locked
  twistLimitDegrees=5
  twistMotion=Locked
 }
 {
  bone1Name=Dyn_skirt_fr_l_02
  bone2Name=Dyn_skirt_fr_l_01
  swing1LimitDegrees=10
  swing2LimitDegrees=2
  swing2Motion=Locked
  twistLimitDegrees=5
  twistMotion=Locked
 }
 {
  bone1Name=Dyn_skirt_btw_fr_l_02
  bone2Name=Dyn_skirt_btw_fr_l_01
  swing1LimitDegrees=10
  swing2LimitDegrees=2
  swing2Motion=Locked
  twistLimitDegrees=5
  twistMotion=Locked
 }
 {
  bone1Name=Base_skirt_btw_fr_l
  bone2Name=pelvis
  swing1LimitDegrees=30
  swing1Motion=Free
  swing2LimitDegrees=30
  swing2Motion=Free
  twistLimitDegrees=10
  twistMotion=Free
 }
 {
  bone1Name=Dyn_skirt_btw_fr_l_01
  bone2Name=Base_skirt_btw_fr_l
  swing1LimitDegrees=30
  swing2LimitDegrees=2
  swing2Motion=Locked
  twistLimitDegrees=5
  twistMotion=Locked
 }
 {
  bone1Name=Base_skirt_mid_fr
  bone2Name=pelvis
  swing1LimitDegrees=45
  swing1Motion=Free
  swing2LimitDegrees=45
  swing2Motion=Free
  twistLimitDegrees=45
  twistMotion=Free
 }
 {
  bone1Name=Dyn_skirt_mid_fr_01
  bone2Name=Base_skirt_mid_fr
  swing1LimitDegrees=15
  swing2LimitDegrees=10
  twistLimitDegrees=45
  twistMotion=Locked
 }
 {
  bone1Name=Dyn_skirt_mid_fr_02
  bone2Name=Dyn_skirt_mid_fr_01
  swing1LimitDegrees=10
  swing2LimitDegrees=5
  swing2Motion=Locked
  twistLimitDegrees=45
  twistMotion=Locked
 }
 {
  bone1Name=Base_skirt_btw_fr_r
  bone2Name=pelvis
  swing1LimitDegrees=45
  swing1Motion=Free
  swing2LimitDegrees=45
  swing2Motion=Free
  twistLimitDegrees=45
  twistMotion=Free
 }
 {
  bone1Name=Dyn_skirt_btw_fr_r_01
  bone2Name=Base_skirt_btw_fr_r
  swing1LimitDegrees=45
  swing1Motion=Free
  swing2LimitDegrees=45
  swing2Motion=Free
  twistLimitDegrees=45
  twistMotion=Free
 }
]
