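# Animation Blueprint for the M_MED_BeaverStage male body cosmetic, adding secondary motion to its cape and skirt.
# It copies the attached parent's pose, runs a Control Rig fed by movement states, then constrains cape bases to spine and shoulders.
# AnimDynamics springs and a pelvis-space RigidBody simulation handle the skirt, while Trail controllers with additive offsets drive cape chains.
# Cached poses are selected by rigid-body, AnimDynamics, DBNO and back-bling flags.
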AnimBlueprint /BRCosmetics/Characters/Player/Male/Medium/Bodies/M_MED_BeaverStage/Meshes/M_MED_BeaverStage_Skeleton_AnimBlueprint
target_skeleton=&/M_MED_BeaverStage_Skeleton
graphs [
 AnimGraph {nodes [
   RigidBody!(
    ComponentPose
    Alpha
    OverrideWorldGravity:Vector="0,0,-980"
    SimSpaceSettings:Sim Space Settings Structure=(WorldAlpha=0,VelocityScaleZ=1,DampingAlpha=1,MaxLinearVelocity=10000,MaxAngularVelocity=10000,MaxLinearAcceleration=10000,MaxAngularAcceleration=10000,ExternalLinearDragV=(X=0,Y=0,Z=0),ExternalLinearVelocity=(X=0,Y=0,Z=0),ExternalAngularVelocity=(X=0,Y=0,Z=0))
   ) -> (Pose) {
    bOverrideWorldGravity=true
    overridePhysicsAsset=&/M_MED_BeaverStage_Physics_skirt
    simulationSpace=BaseBoneSpace
    alphaBoolBlend {blendOption=Linear}
    alphaScaleBiasClamp {interpSpeedDecreasing=7, interpSpeedIncreasing=15}
    baseBoneRef {boneName=pelvis}
   }
   n1 = Blend Poses by bool(BlendPose_0=Use cached pose 'dbnoCache'(), BlendPose_1=Use cached pose 'trailCache_1'(), bActiveValue=Get bIsDBNO()) {blendType=Linear}
   n2 = Use cached pose 'AnimDyn Cache'()
   n3 = Get bAnimDynamicsEnabled() -> (bAnimDynamicsEnabled:bool)
   Output Pose(Result=Blend Poses by bool(
     BlendPose_0=Use cached pose 'Rban Cache'()
     BlendPose_1=Blend Poses by bool(BlendPose_0=@n2, BlendPose_1=@n1, bActiveValue=@n3) {blendType=Linear}
     bActiveValue=Get bRigidBodyEnabled()
    ) {blendType=Linear})
   n4 = Get TrailControllerRot() -> (TrailControllerRot:Rotator)
   n5 = "Transform (Modify) Bone - Bone: cape_mid_02_l"(ComponentPose=Local To Component(LocalPose=Use cached pose 'Rban Cache'()), Alpha=1, Translation="0,0,0", Rotation=@n4, Scale="1,1,1") {rotationMode=BMM_Additive, rotationSpace=BCS_BoneSpace, alphaBoolBlend {blendOption=Linear}, boneToModify {boneName=cape_mid_02_l}}
   n6 = "Transform (Modify) Bone - Bone: cape_mid_02_r"(ComponentPose=@n5, Alpha=1, Translation="0,0,0", Rotation=@n4, Scale="1,1,1") {rotationMode=BMM_Additive, rotationSpace=BCS_BoneSpace, alphaBoolBlend {blendOption=Linear}, boneToModify {boneName=cape_mid_02_r}}
   n7 = "Transform (Modify) Bone - Bone: cape_03_l"(ComponentPose=@n6, Alpha=1, Translation="0,0,0", Rotation=@n4, Scale="1,1,1") {rotationMode=BMM_Additive, rotationSpace=BCS_BoneSpace, alphaBoolBlend {blendOption=Linear}, boneToModify {boneName=cape_03_l}}
   n8 = "Transform (Modify) Bone - Bone: cape_03_r"(ComponentPose=@n7, Alpha=1, Translation="0,0,0", Rotation=@n4, Scale="1,1,1") {rotationMode=BMM_Additive, rotationSpace=BCS_BoneSpace, alphaBoolBlend {blendOption=Linear}, boneToModify {boneName=cape_03_r}}
   n9 = Get BwdMovementTrail() -> (BwdMovementTrail:double)
   n10 = Trail controller - Bone: cape_mid_06_l(ComponentPose=@n8, Alpha=@n9) {
    bUsePlanarLimit=true
    chainLength=4
    debugLifeTime=0.1
    maxDeltaTime=0.333
    alphaScaleBiasClamp {bInterpResult=true}
    rotationLimits [
     {limitMax=<180 20 180>, limitMin=<-180 -70 -180>}
     {limitMax=<180 20 180>, limitMin=<-180 -70 -180>}
     {limitMax=<180 20 180>, limitMin=<-180 -70 -180>}
     {limitMax=<180 20 180>, limitMin=<-180 -70 -180>}
    ]
    rotationOffsets [<0 0 0>, <0 0 0>, <0 0 0>, <0 0 0>]
    trailBone {boneName=cape_mid_06_l}
    trailRelaxationSpeed {editorCurveData {keys [<at -0.0134 19.3841 RCIM_Linear>,<at 0.5079 12.8986 RCIM_Linear>,<at 0.9955 9.6044 RCIM_Linear>]}}
   }
   n11 = Trail controller - Bone: cape_mid_06_r(ComponentPose=@n10, Alpha=@n9) {
    bUsePlanarLimit=true
    chainLength=4
    debugLifeTime=0.1
    maxDeltaTime=0.0333
    alphaScaleBiasClamp {bInterpResult=true}
    rotationLimits [
     {limitMax=<180 20 180>, limitMin=<-180 -70 -180>}
     {limitMax=<180 20 180>, limitMin=<-180 -70 -180>}
     {limitMax=<180 20 180>, limitMin=<-180 -70 -180>}
     {limitMax=<180 20 180>, limitMin=<-180 -70 -180>}
    ]
    rotationOffsets [<0 0 0>, <0 0 0>, <0 0 0>, <0 0 0>]
    trailBone {boneName=cape_mid_06_r}
    trailRelaxationSpeed {editorCurveData {keys [<at 0 15 RCIM_Linear>,<at 0.5289 13.1838 RCIM_Linear>,<at 1 6 RCIM_Linear>]}}
   }
   n12 = Trail controller - Bone: cape_07_l(ComponentPose=@n11, Alpha=@n9) {
    bUsePlanarLimit=true
    chainLength=4
    debugLifeTime=0.1
    maxDeltaTime=0.0333
    alphaScaleBiasClamp {bInterpResult=true}
    rotationLimits [
     {limitMax=<180 20 180>, limitMin=<-180 -70 -180>}
     {limitMax=<180 20 180>, limitMin=<-180 -70 -180>}
     {limitMax=<180 20 180>, limitMin=<-180 -70 -180>}
     {limitMax=<180 20 180>, limitMin=<-180 -70 -180>}
    ]
    rotationOffsets [<0 0 0>, <0 0 0>, <0 0 0>, <0 0 0>]
    trailBone {boneName=cape_07_l}
    trailRelaxationSpeed {editorCurveData {keys [<at 0 15 RCIM_Linear>,<at 0.5289 13.1838 RCIM_Linear>,<at 1 6 RCIM_Linear>]}}
   }
   n13 = Trail controller - Bone: cape_07_r(ComponentPose=@n12, Alpha=@n9) {
    bUsePlanarLimit=true
    chainLength=4
    debugLifeTime=0.1
    maxDeltaTime=0.0333
    alphaScaleBiasClamp {bInterpResult=true}
    rotationLimits [
     {limitMax=<180 20 180>, limitMin=<-180 -70 -180>}
     {limitMax=<180 20 180>, limitMin=<-180 -70 -180>}
     {limitMax=<180 20 180>, limitMin=<-180 -70 -180>}
     {limitMax=<180 20 180>, limitMin=<-180 -70 -180>}
    ]
    rotationOffsets [<0 0 0>, <0 0 0>, <0 0 0>, <0 0 0>]
    trailBone {boneName=cape_07_r}
    trailRelaxationSpeed {editorCurveData {keys [<at 0 15 RCIM_Linear>,<at 0.5289 13.1838 RCIM_Linear>,<at 1 6 RCIM_Linear>]}}
   }
   n14 = "Transform (Modify) Bone - Bone: cape_mid_01_l_MSTR"(ComponentPose=@n13, Alpha=1, Translation="0,0,0", Rotation="-10,0,0", Scale="1,1,1") {
    rotationMode=BMM_Additive
    rotationSpace=BCS_BoneSpace
    alphaBoolBlend {blendOption=Linear}
    boneToModify {boneName=cape_mid_01_l_MSTR}
    rotation {pitch=-10}
   }
   n15 = "Transform (Modify) Bone - Bone: cape_mid_01_r_MSTR"(ComponentPose=@n14, Alpha=1, Translation="0,0,0", Rotation="-10,0,0", Scale="1,1,1") {
    rotationMode=BMM_Additive
    rotationSpace=BCS_BoneSpace
    alphaBoolBlend {blendOption=Linear}
    boneToModify {boneName=cape_mid_01_r_MSTR}
    rotation {pitch=-10}
   }
   n16 = "Transform (Modify) Bone - Bone: cape_01_l"(ComponentPose=@n15, Alpha=1, Translation="0,0,0", Rotation="-10,0,0", Scale="1,1,1") {
    rotationMode=BMM_Additive
    rotationSpace=BCS_BoneSpace
    alphaBoolBlend {blendOption=Linear}
    boneToModify {boneName=cape_01_l}
    rotation {pitch=-10}
   }
   n17 = "Transform (Modify) Bone - Bone: cape_01_r"(ComponentPose=@n16, Alpha=1, Translation="0,0,0", Rotation="-10,0,0", Scale="1,1,1") {
    rotationMode=BMM_Additive
    rotationSpace=BCS_BoneSpace
    alphaBoolBlend {blendOption=Linear}
    boneToModify {boneName=cape_01_r}
    rotation {pitch=-10}
   }
   n18 = "Transform (Modify) Bone - Bone: cape_mid_03_l"!(ComponentPose=@n17, bAlphaBoolEnabled:bool=True, Translation="0,0,0", Rotation="10,0,0", Scale="1,1,1") -> (Pose) {
    alphaInputType=Bool
    rotationMode=BMM_Additive
    rotationSpace=BCS_BoneSpace
    alphaBoolBlend {blendOption=Linear}
    boneToModify {boneName=cape_mid_03_l}
    rotation {pitch=10}
   }
   n19 = "Transform (Modify) Bone - Bone: cape_mid_03_r"!(ComponentPose=@n18, bAlphaBoolEnabled:bool=True, Translation="0,0,0", Rotation="10,0,0", Scale="1,1,1") -> (Pose) {
    alphaInputType=Bool
    rotationMode=BMM_Additive
    rotationSpace=BCS_BoneSpace
    alphaBoolBlend {blendOption=Linear}
    boneToModify {boneName=cape_mid_03_r}
    rotation {pitch=10}
   }
   n20 = "Transform (Modify) Bone - Bone: cape_04_l"!(ComponentPose=@n19, bAlphaBoolEnabled:bool=True, Translation="0,0,0", Rotation="10,0,0", Scale="1,1,1") -> (Pose) {
    alphaInputType=Bool
    rotationMode=BMM_Additive
    rotationSpace=BCS_BoneSpace
    alphaBoolBlend {blendOption=Linear}
    boneToModify {boneName=cape_04_l}
    rotation {pitch=10}
   }
   n21 = "Transform (Modify) Bone - Bone: cape_04_r"!(ComponentPose=@n20, bAlphaBoolEnabled:bool=True, Translation="0,0,0", Rotation="10,0,0", Scale="1,1,1") -> (Pose) {
    alphaInputType=Bool
    rotationMode=BMM_Additive
    rotationSpace=BCS_BoneSpace
    alphaBoolBlend {blendOption=Linear}
    boneToModify {boneName=cape_04_r}
    rotation {pitch=10}
   }
   n22 = Get TrailControllerRot()
   n23 = "Transform (Modify) Bone - Bone: cape_mid_01_l_MSTR"(ComponentPose=Local To Component(LocalPose=Use cached pose 'Rban Cache'()), Alpha=1, Translation="0,0,0", Rotation=@n22, Scale="1,1,1") {rotationMode=BMM_Additive, rotationSpace=BCS_BoneSpace, alphaBoolBlend {blendOption=Linear}, boneToModify {boneName=cape_mid_01_l_MSTR}}
   n24 = "Transform (Modify) Bone - Bone: cape_mid_01_r_MSTR"(ComponentPose=@n23, Alpha=1, Translation="0,0,0", Rotation=@n22, Scale="1,1,1") {rotationMode=BMM_Additive, rotationSpace=BCS_BoneSpace, alphaBoolBlend {blendOption=Linear}, boneToModify {boneName=cape_mid_01_r_MSTR}}
   n25 = "Transform (Modify) Bone - Bone: cape_01_l"(ComponentPose=@n24, Alpha=1, Translation="0,0,0", Rotation=@n22, Scale="1,1,1") {rotationMode=BMM_Additive, rotationSpace=BCS_BoneSpace, alphaBoolBlend {blendOption=Linear}, boneToModify {boneName=cape_01_l}}
   n26 = "Transform (Modify) Bone - Bone: cape_01_r"(ComponentPose=@n25, Alpha=1, Translation="0,0,0", Rotation=@n22, Scale="1,1,1") {rotationMode=BMM_Additive, rotationSpace=BCS_BoneSpace, alphaBoolBlend {blendOption=Linear}, boneToModify {boneName=cape_01_r}}
   n27 = Get BwdMovementTrail()
   n28 = Trail controller - Bone: cape_mid_06_l(ComponentPose=@n26, Alpha=@n27) {
    bUsePlanarLimit=true
    chainLength=5
    debugLifeTime=0.1
    maxDeltaTime=0.333
    alphaScaleBiasClamp {bInterpResult=true}
    rotationLimits [{limitMax=<180 180 180>}, {limitMax=<180 180 180>}, {limitMax=<180 180 180>}, {limitMax=<180 180 180>}, {limitMax=<180 180 180>}]
    rotationOffsets [<0 0 0>, <0 0 0>, <0 0 0>, <0 0 0>, <0 0 0>]
    trailBone {boneName=cape_mid_06_l}
    trailRelaxationSpeed {editorCurveData {keys [<at -0.0134 19.3841 RCIM_Linear>,<at 0.5079 12.8986 RCIM_Linear>,<at 0.9955 9.6044 RCIM_Linear>]}}
   }
   n29 = Trail controller - Bone: cape_mid_06_r(ComponentPose=@n28, Alpha=@n27) {
    bUsePlanarLimit=true
    chainLength=5
    debugLifeTime=0.1
    maxDeltaTime=0.0333
    alphaScaleBiasClamp {bInterpResult=true}
    rotationLimits [
     {limitMax=<180 180 180>}
     {limitMax=<180 180 180>}
     {limitMax=<180 90 180>, limitMin=<-180 -90 -180>}
     {limitMax=<180 180 180>}
     {limitMax=<180 180 180>}
    ]
    rotationOffsets [<0 0 0>, <0 0 0>, <0 0 0>, <0 0 0>, <0 0 0>]
    trailBone {boneName=cape_mid_06_r}
    trailRelaxationSpeed {editorCurveData {keys [<at 0 15 RCIM_Linear>,<at 0.5289 13.1838 RCIM_Linear>,<at 1 6 RCIM_Linear>]}}
   }
   n30 = Trail controller - Bone: cape_07_l(ComponentPose=@n29, Alpha=@n27) {
    bUsePlanarLimit=true
    chainLength=5
    debugLifeTime=0.1
    maxDeltaTime=0.0333
    alphaScaleBiasClamp {bInterpResult=true}
    rotationLimits [
     {limitMax=<180 180 180>}
     {limitMax=<180 180 180>}
     {limitMax=<180 90 180>, limitMin=<-180 -90 -180>}
     {limitMax=<180 180 180>}
     {limitMax=<180 180 180>}
    ]
    rotationOffsets [<0 0 0>, <0 0 0>, <0 0 0>, <0 0 0>, <0 0 0>]
    trailBone {boneName=cape_07_l}
    trailRelaxationSpeed {editorCurveData {keys [<at 0 15 RCIM_Linear>,<at 0.5289 13.1838 RCIM_Linear>,<at 1 6 RCIM_Linear>]}}
   }
   n31 = Trail controller - Bone: cape_07_r(ComponentPose=@n30, Alpha=@n27) {
    bUsePlanarLimit=true
    chainLength=5
    debugLifeTime=0.1
    maxDeltaTime=0.0333
    alphaScaleBiasClamp {bInterpResult=true}
    rotationLimits [
     {limitMax=<180 180 180>}
     {limitMax=<180 180 180>}
     {limitMax=<180 90 180>, limitMin=<-180 -90 -180>}
     {limitMax=<180 180 180>}
     {limitMax=<180 180 180>}
    ]
    rotationOffsets [<0 0 0>, <0 0 0>, <0 0 0>, <0 0 0>, <0 0 0>]
    trailBone {boneName=cape_07_r}
    trailRelaxationSpeed {editorCurveData {keys [<at 0 15 RCIM_Linear>,<at 0.5289 13.1838 RCIM_Linear>,<at 1 6 RCIM_Linear>]}}
   }
   Save cached pose 'trailCache_1'(Pose=Blend Poses by bool(
     BlendPose_0=Component To Local(ComponentPose=@n21)
     BlendPose_1=Component To Local(ComponentPose=@n31)
     bActiveValue=Get IsWearingBackbling() -> (IsWearingBackbling:bool)
    )) {cacheName=trailCache_1}
   n32 = "Transform (Modify) Bone - Bone: cape_01_l"(ComponentPose=Local To Component(LocalPose=Use cached pose 'Start Cache'()), Alpha=1, Translation="0,0,0", Rotation="5,0,0", Scale="1,1,1") {
    rotationMode=BMM_Additive
    rotationSpace=BCS_BoneSpace
    alphaBoolBlend {blendOption=Linear}
    boneToModify {boneName=cape_01_l}
    rotation {pitch=5}
   }
   n33 = "Transform (Modify) Bone - Bone: cape_03_l"(ComponentPose=@n32, Alpha=1, Translation="0,0,0", Rotation="-10,0,0", Scale="1,1,1") {
    rotationMode=BMM_Additive
    rotationSpace=BCS_BoneSpace
    alphaBoolBlend {blendOption=Linear}
    boneToModify {boneName=cape_03_l}
    rotation {pitch=-10}
   }
   n34 = "Transform (Modify) Bone - Bone: cape_06_l"(ComponentPose=@n33, Alpha=1, Translation="0,0,0", Rotation="-25,0,0", Scale="1,1,1") {
    rotationMode=BMM_Additive
    rotationSpace=BCS_BoneSpace
    alphaBoolBlend {blendOption=Linear}
    boneToModify {boneName=cape_06_l}
    rotation {pitch=-25}
   }
   n35 = Trail controller - Bone: cape_mid_06_l!(ComponentPose=@n34, Alpha=1, RelaxationSpeedScale:float=2, FakeVelocity:Vector="0,0,0") -> (Pose) {
    bActorSpaceFakeVel=true
    bShowTrailLocation=true
    bUsePlanarLimit=true
    chainLength=5
    debugLifeTime=0.1
    maxDeltaTime=0.0333
    relaxationSpeedScale=1.2
    alphaScaleBiasClamp {bInterpResult=true, interpSpeedDecreasing=25, interpSpeedIncreasing=25}
    baseJoint {boneName=backpack_root}
    rotationLimits [{limitMax=<180 180 180>}, {limitMax=<180 180 180>}, {limitMax=<180 180 180>}, {limitMax=<180 180 180>}, {limitMax=<180 180 180>}]
    rotationOffsets [<0 0 0>, <0 0 0>, <0 0 0>, <0 0 0>, <0 0 0>]
    trailBone {boneName=cape_mid_06_l}
    trailRelaxationSpeed {editorCurveData {keys [<at 0 15 RCIM_Linear>,<at 1 6 RCIM_Linear>]}}
   }
   n36 = Trail controller - Bone: cape_mid_06_r!(ComponentPose=@n35, Alpha=1, RelaxationSpeedScale:float=2, FakeVelocity:Vector="0,0,0") -> (Pose) {
    bActorSpaceFakeVel=true
    bShowTrailLocation=true
    bUsePlanarLimit=true
    chainLength=5
    debugLifeTime=0.1
    maxDeltaTime=0.0333
    relaxationSpeedScale=1.2
    alphaScaleBiasClamp {bInterpResult=true, interpSpeedDecreasing=25, interpSpeedIncreasing=25}
    baseJoint {boneName=backpack_root}
    rotationLimits [{limitMax=<180 180 180>}, {limitMax=<180 180 180>}, {limitMax=<180 180 180>}, {limitMax=<180 180 180>}, {limitMax=<180 180 180>}]
    rotationOffsets [<0 0 0>, <0 0 0>, <0 0 0>, <0 0 0>, <0 0 0>]
    trailBone {boneName=cape_mid_06_r}
    trailRelaxationSpeed {editorCurveData {keys [<at 0 15 RCIM_Linear>,<at 1 6 RCIM_Linear>]}}
   }
   n37 = Trail controller - Bone: cape_07_l!(ComponentPose=@n36, Alpha=1, RelaxationSpeedScale:float=2, FakeVelocity:Vector="0,0,0") -> (Pose) {
    bActorSpaceFakeVel=true
    bShowTrailLocation=true
    bUsePlanarLimit=true
    chainLength=6
    debugLifeTime=0.1
    maxDeltaTime=0.0333
    relaxationSpeedScale=1.2
    alphaScaleBiasClamp {bInterpResult=true, interpSpeedDecreasing=25, interpSpeedIncreasing=25}
    baseJoint {boneName=backpack_root}
    rotationLimits [
     {limitMax=<180 180 180>}
     {limitMax=<180 180 180>}
     {limitMax=<180 180 180>}
     {limitMax=<180 180 180>}
     {limitMax=<180 180 180>}
     {limitMax=<180 180 180>}
    ]
    rotationOffsets [<0 0 0>, <0 0 0>, <0 0 0>, <0 0 0>, <0 0 0>, <0 0 0>]
    trailBone {boneName=cape_07_l}
    trailRelaxationSpeed {editorCurveData {keys [<at 0 15 RCIM_Linear>,<at 1 6 RCIM_Linear>]}}
   }
   n38 = Trail controller - Bone: cape_07_r!(ComponentPose=@n37, Alpha=1, RelaxationSpeedScale:float=2, FakeVelocity:Vector="0,0,0") -> (Pose) {
    bActorSpaceFakeVel=true
    bShowTrailLocation=true
    bUsePlanarLimit=true
    chainLength=6
    debugLifeTime=0.1
    maxDeltaTime=0.0333
    relaxationSpeedScale=1.2
    alphaScaleBiasClamp {bInterpResult=true, interpSpeedDecreasing=25, interpSpeedIncreasing=25}
    baseJoint {boneName=backpack_root}
    rotationLimits [
     {limitMax=<180 180 180>}
     {limitMax=<180 180 180>}
     {limitMax=<180 180 180>}
     {limitMax=<180 180 180>}
     {limitMax=<180 180 180>}
     {limitMax=<180 180 180>}
    ]
    rotationOffsets [<0 0 0>, <0 0 0>, <0 0 0>, <0 0 0>, <0 0 0>, <0 0 0>]
    trailBone {boneName=cape_07_r}
    trailRelaxationSpeed {editorCurveData {keys [<at 0 15 RCIM_Linear>,<at 1 6 RCIM_Linear>]}}
   }
   Save cached pose 'dbnoCache'(Pose=Component To Local(ComponentPose=@n38)) {cacheName=dbnoCache}
   n39 = Control Rig!(
    Source=Copy Pose From Mesh() {bUseAttachedParent=true}
    Alpha
    Is Right Left Movement:double=Get Is Right Left Movement() -> (Is Right Left Movement:double)
    Is Backward Forward Movement:double=Get Is Forward Backward Movement() -> (Is Forward Backward Movement:double)
    Jumping ON:bool=Get Is Jumping() -> (Is Jumping:bool)
    Is Skydive:bool=Get Sky() -> (Sky:bool)
    BaseballSliding:double=Get BaseballSliding() -> (BaseballSliding:double)
    PickaxeSwing:double=Get Picaxeswing() -> (Picaxeswing:double)
    Weight:bool=Get bIsInFrontEnd()
    LOD2:int=Get CurrentLOD()
    IsPlayingEmote:bool=Get bIsPlayingEmote()
    DivingDown:bool=Get DivingDown() -> (DivingDown:bool)
    Swimming:bool=false
   ) -> (Pose) {controlRigAssetReference {controlRigAsset=&/M_MED_BeaverStage_CtrlRig}}
   n40 = "Constraint - Base_Snap_cape_r (PPPPPPPPP)"!(
    ComponentPose=Local To Component(LocalPose=@n39)
    Alpha=1
    ConstraintWeights_0=0.0045
    ConstraintWeights_1:float=0.0651
    ConstraintWeights_2:float=0.1543
    ConstraintWeights_3:float=0.2438
    ConstraintWeights_4:float=0.3561
    ConstraintWeights_5:float=0.1362
    ConstraintWeights_6:float=0.0019
    ConstraintWeights_7:float=0.0324
    ConstraintWeights_8:float=0.0057
   ) -> (Pose) {
    boneToModify {boneName=Base_Snap_cape_r}
    constraintSetup [
     {targetBone {boneName=spine_01}}
     {targetBone {boneName=spine_02}}
     {targetBone {boneName=spine_03}}
     {targetBone {boneName=spine_04}}
     {targetBone {boneName=spine_05}}
     {targetBone {boneName=clavicle_r}}
     {targetBone {boneName=upperarm_twist_01_r}}
     {targetBone {boneName=deform_upperarm_fr_r}}
     {targetBone {boneName=deform_pec_r}}
    ]
    constraintWeights [1,1,1,1,1,1,1,1,1]
   }
   n41 = "Constraint - Base_Snap_cape_l (PPPPPPPPP)"!(
    ComponentPose=@n40
    Alpha=1
    ConstraintWeights_0=0.0045
    ConstraintWeights_1:float=0.0651
    ConstraintWeights_2:float=0.1543
    ConstraintWeights_3:float=0.2438
    ConstraintWeights_4:float=0.3561
    ConstraintWeights_5:float=0.1362
    ConstraintWeights_6:float=0.0019
    ConstraintWeights_7:float=0.0324
    ConstraintWeights_8:float=0.0057
   ) -> (Pose) {
    boneToModify {boneName=Base_Snap_cape_l}
    constraintSetup [
     {targetBone {boneName=spine_01}}
     {targetBone {boneName=spine_02}}
     {targetBone {boneName=spine_03}}
     {targetBone {boneName=spine_04}}
     {targetBone {boneName=spine_05}}
     {targetBone {boneName=clavicle_l}}
     {targetBone {boneName=upperarm_twist_01_l}}
     {targetBone {boneName=deform_upperarm_fr_l}}
     {targetBone {boneName=deform_pec_l}}
    ]
    constraintWeights [1,1,1,1,1,1,1,1,1]
   }
   n42 = "Transform (Modify) Bone - Bone: attach_backpack"(ComponentPose=@n41, Alpha=1, Translation="0,-9.701,0.5", Rotation="0,0,12", Scale="1,1,1") {
    rotationMode=BMM_Additive
    rotationSpace=BCS_BoneSpace
    translationMode=BMM_Additive
    translationSpace=BCS_BoneSpace
    boneToModify {boneName=attach_backpack}
    rotation {roll=12}
    translation {y=-9.701, z=0.5}
   }
   n43 = Anim Dynamics - Bone: Dyn_skirt_btw_fr_l_01!(ComponentPose=@n42, bAlphaBoolEnabled:bool=True, ExternalForce="0,0,0") -> (Pose) {
    alphaInputType=Bool
    angularDampingOverride=0.9
    angularSpringConstant=800
    bAngularSpring=true
    bGravityOverrideInSimSpace=true
    bOverrideAngularDamping=true
    bUseGravityOverride=true
    bUsePlanarLimit=false
    simulationSpace=RootRelative
    worldSpaceMinimumScale=0.022
    boundBone {boneName=Dyn_skirt_btw_fr_l_01}
    componentAppliedLinearAccClamp=<3000 3000 3000>
    componentLinearAccScale=<0.6 0.6 0.6>
    componentLinearVelScale=<0.6 0.6 0.6>
    gravityOverride {z=-980}
    physicsBodyDefinitions [{
      boundBone {boneName=Dyn_skirt_btw_fr_l_01}
      boxExtents=<12 12 12>
      constraintSetup {coneAngle=1, angularLimitsMax=<0 0 33>, angularLimitsMin=<0 0 2>, angularTarget=<1 0 0>}
      localJointOffset=<0.4 0 0>
     }]
    simSpaceSettings {maxAngularAcceleration=3000, maxAngularVelocity=3000, simSpaceAngularAlpha=0.5}
   }
   n44 = Anim Dynamics - Bone: Dyn_skirt_btw_fr_l_02!(ComponentPose=@n43, bAlphaBoolEnabled:bool=True, ExternalForce="0,0,0") -> (Pose) {
    alphaInputType=Bool
    angularDampingOverride=0.9
    angularSpringConstant=800
    bAngularSpring=true
    bGravityOverrideInSimSpace=true
    bOverrideAngularDamping=true
    bUseGravityOverride=true
    bUsePlanarLimit=false
    simulationSpace=RootRelative
    worldSpaceMinimumScale=0.022
    boundBone {boneName=Dyn_skirt_btw_fr_l_02}
    componentAppliedLinearAccClamp=<3000 3000 3000>
    componentLinearAccScale=<0.6 0.6 0.6>
    componentLinearVelScale=<0.6 0.6 0.6>
    gravityOverride {z=-980}
    physicsBodyDefinitions [{
      boundBone {boneName=Dyn_skirt_btw_fr_l_02}
      boxExtents=<12 12 12>
      constraintSetup {coneAngle=1, angularLimitsMax=<0 0 33>, angularLimitsMin=<0 0 2>, angularTarget=<1 0 0>}
      localJointOffset=<0.4 0 0>
     }]
    simSpaceSettings {maxAngularAcceleration=3000, maxAngularVelocity=3000, simSpaceAngularAlpha=0.5}
   }
   n45 = Anim Dynamics - Bone: Dyn_skirt_mid_fr_01!(ComponentPose=@n44, bAlphaBoolEnabled:bool=True, ExternalForce="0,0,0") -> (Pose) {
    alphaInputType=Bool
    angularDampingOverride=0.9
    angularSpringConstant=800
    bAngularSpring=true
    bGravityOverrideInSimSpace=true
    bOverrideAngularDamping=true
    bUseGravityOverride=true
    bUsePlanarLimit=false
    simulationSpace=RootRelative
    worldSpaceMinimumScale=0.022
    boundBone {boneName=Dyn_skirt_mid_fr_01}
    componentAppliedLinearAccClamp=<3000 3000 3000>
    componentLinearAccScale=<0.6 0.6 0.6>
    componentLinearVelScale=<0.6 0.6 0.6>
    gravityOverride {z=-980}
    physicsBodyDefinitions [{
      boundBone {boneName=Dyn_skirt_mid_fr_01}
      boxExtents=<12 12 12>
      constraintSetup {coneAngle=1, angularLimitsMax=<0 0 33>, angularLimitsMin=<0 0 -5>, angularTarget=<1 0 0>}
      localJointOffset=<0.5 0 0>
     }]
    simSpaceSettings {maxAngularAcceleration=3000, maxAngularVelocity=3000, simSpaceAngularAlpha=0.5}
   }
   n46 = Anim Dynamics - Bone: Dyn_skirt_mid_fr_02!(ComponentPose=@n45, bAlphaBoolEnabled:bool=True, ExternalForce="0,0,0") -> (Pose) {
    alphaInputType=Bool
    angularDampingOverride=0.9
    angularSpringConstant=800
    bAngularSpring=true
    bGravityOverrideInSimSpace=true
    bOverrideAngularDamping=true
    bUseGravityOverride=true
    bUsePlanarLimit=false
    simulationSpace=RootRelative
    worldSpaceMinimumScale=0.022
    boundBone {boneName=Dyn_skirt_mid_fr_02}
    componentAppliedLinearAccClamp=<3000 3000 3000>
    componentLinearAccScale=<0.6 0.6 0.6>
    componentLinearVelScale=<0.6 0.6 0.6>
    gravityOverride {z=-980}
    physicsBodyDefinitions [{
      boundBone {boneName=Dyn_skirt_mid_fr_02}
      boxExtents=<12 12 12>
      constraintSetup {coneAngle=1, angularLimitsMax=<0 0 33>, angularLimitsMin=<0 0 1>, angularTarget=<1 0 0>}
      localJointOffset=<0.5 0 0>
     }]
    simSpaceSettings {maxAngularAcceleration=3000, maxAngularVelocity=3000, simSpaceAngularAlpha=0.5}
   }
   n47 = Anim Dynamics - Bone: Dyn_skirt_fr_l_01!(ComponentPose=@n46, bAlphaBoolEnabled:bool=True, ExternalForce="0,0,0") -> (Pose) {
    alphaInputType=Bool
    angularDampingOverride=0.9
    angularSpringConstant=800
    bAngularSpring=true
    bGravityOverrideInSimSpace=true
    bOverrideAngularDamping=true
    bUseGravityOverride=true
    bUsePlanarLimit=false
    simulationSpace=RootRelative
    worldSpaceMinimumScale=0.022
    boundBone {boneName=Dyn_skirt_fr_l_01}
    componentAppliedLinearAccClamp=<3000 3000 3000>
    componentLinearAccScale=<0.6 0.6 0.6>
    componentLinearVelScale=<0.6 0.6 0.6>
    gravityOverride {z=-980}
    physicsBodyDefinitions [{
      boundBone {boneName=Dyn_skirt_fr_l_01}
      boxExtents=<12 12 12>
      constraintSetup {coneAngle=1, angularLimitsMax=<0 0 33>, angularLimitsMin=<0 0 1>, angularTarget=<1 0 0>}
      localJointOffset=<0.5 0 0>
     }]
    simSpaceSettings {maxAngularAcceleration=3000, maxAngularVelocity=3000, simSpaceAngularAlpha=0.5}
   }
   n48 = Anim Dynamics - Bone: Dyn_skirt_fr_l_02!(ComponentPose=@n47, bAlphaBoolEnabled:bool=True, ExternalForce="0,0,0") -> (Pose) {
    alphaInputType=Bool
    angularDampingOverride=0.9
    angularSpringConstant=800
    bAngularSpring=true
    bGravityOverrideInSimSpace=true
    bOverrideAngularDamping=true
    bUseGravityOverride=true
    bUsePlanarLimit=false
    simulationSpace=RootRelative
    worldSpaceMinimumScale=0.022
    boundBone {boneName=Dyn_skirt_fr_l_02}
    componentAppliedLinearAccClamp=<3000 3000 3000>
    componentLinearAccScale=<0.6 0.6 0.6>
    componentLinearVelScale=<0.6 0.6 0.6>
    gravityOverride {z=-980}
    physicsBodyDefinitions [{
      boundBone {boneName=Dyn_skirt_fr_l_02}
      boxExtents=<12 12 12>
      constraintSetup {coneAngle=1, angularLimitsMax=<0 0 33>, angularLimitsMin=<0 0 1>, angularTarget=<1 0 0>}
      localJointOffset=<0.5 0 0>
     }]
    simSpaceSettings {maxAngularAcceleration=3000, maxAngularVelocity=3000, simSpaceAngularAlpha=0.5}
   }
   n49 = Anim Dynamics - Bone: Dyn_skirt_fr_r_01(ComponentPose=@n48, Alpha=1, ExternalForce="0,0,0") {
    angularDampingOverride=0.9
    angularSpringConstant=800
    bAngularSpring=true
    bGravityOverrideInSimSpace=true
    bOverrideAngularDamping=true
    bUseGravityOverride=true
    bUsePlanarLimit=false
    simulationSpace=RootRelative
    worldSpaceMinimumScale=0.022
    boundBone {boneName=Dyn_skirt_fr_r_01}
    componentAppliedLinearAccClamp=<3000 3000 3000>
    componentLinearAccScale=<0.6 0.6 0.6>
    componentLinearVelScale=<0.6 0.6 0.6>
    gravityOverride {z=-980}
    physicsBodyDefinitions [{
      boundBone {boneName=Dyn_skirt_fr_r_01}
      boxExtents=<12 12 12>
      constraintSetup {angularLimitsMax=<0 2 33>, angularLimitsMin=<0 -2 1>, angularTarget=<1 0 0>}
      localJointOffset=<0.4 0 0>
     }]
    simSpaceSettings {maxAngularAcceleration=3000, maxAngularVelocity=3000, simSpaceAngularAlpha=0.5}
   }
   n50 = Anim Dynamics - Bone: Dyn_skirt_fr_r_02(ComponentPose=@n49, Alpha=1, ExternalForce="0,0,0") {
    angularDampingOverride=0.9
    angularSpringConstant=800
    bGravityOverrideInSimSpace=true
    bOverrideAngularDamping=true
    bUseGravityOverride=true
    bUsePlanarLimit=false
    simulationSpace=RootRelative
    worldSpaceMinimumScale=0.022
    boundBone {boneName=Dyn_skirt_fr_r_02}
    componentAppliedLinearAccClamp=<3000 3000 3000>
    componentLinearAccScale=<0.6 0.6 0.6>
    componentLinearVelScale=<0.6 0.6 0.6>
    gravityOverride {z=-980}
    physicsBodyDefinitions [{
      boundBone {boneName=Dyn_skirt_fr_r_02}
      boxExtents=<12 12 12>
      constraintSetup {angularLimitsMax=<0 2 33>, angularLimitsMin=<0 -2 1>, angularTarget=<1 0 0>}
      localJointOffset=<0.4 0 0>
     }]
    simSpaceSettings {maxAngularAcceleration=3000, maxAngularVelocity=3000, simSpaceAngularAlpha=0.5}
   }
   n51 = Anim Dynamics - Bone: Dyn_skirt_btw_fr_r_01(ComponentPose=@n50, Alpha=1, ExternalForce="0,0,0") {
    angularDampingOverride=0.9
    angularSpringConstant=800
    bAngularSpring=true
    bGravityOverrideInSimSpace=true
    bOverrideAngularDamping=true
    bUseGravityOverride=true
    bUsePlanarLimit=false
    simulationSpace=RootRelative
    worldSpaceMinimumScale=0.022
    boundBone {boneName=Dyn_skirt_btw_fr_r_01}
    componentAppliedLinearAccClamp=<3000 3000 3000>
    componentLinearAccScale=<0.6 0.6 0.6>
    componentLinearVelScale=<0.6 0.6 0.6>
    gravityOverride {z=-980}
    physicsBodyDefinitions [{
      boundBone {boneName=Dyn_skirt_btw_fr_r_01}
      boxExtents=<12 12 12>
      constraintSetup {angularLimitsMax=<0 0 33>, angularLimitsMin=<0 0 2>, angularTarget=<1 0 0>}
      localJointOffset=<0.4 0 0>
     }]
    simSpaceSettings {maxAngularAcceleration=3000, maxAngularVelocity=3000, simSpaceAngularAlpha=0.5}
   }
   n52 = Anim Dynamics - Bone: Dyn_skirt_btw_fr_r_02(ComponentPose=@n51, Alpha=1, ExternalForce="0,0,0") {
    angularDampingOverride=0.9
    angularSpringConstant=800
    bGravityOverrideInSimSpace=true
    bOverrideAngularDamping=true
    bUseGravityOverride=true
    bUsePlanarLimit=false
    simulationSpace=RootRelative
    worldSpaceMinimumScale=0.022
    boundBone {boneName=Dyn_skirt_btw_fr_r_02}
    componentAppliedLinearAccClamp=<3000 3000 3000>
    componentLinearAccScale=<0.6 0.6 0.6>
    componentLinearVelScale=<0.6 0.6 0.6>
    gravityOverride {z=-980}
    physicsBodyDefinitions [{
      boundBone {boneName=Dyn_skirt_btw_fr_r_02}
      boxExtents=<12 12 12>
      constraintSetup {angularLimitsMax=<0 0 33>, angularLimitsMin=<0 0 2>, angularTarget=<1 0 0>}
      localJointOffset=<0.4 0 0>
     }]
    simSpaceSettings {maxAngularAcceleration=3000, maxAngularVelocity=3000, simSpaceAngularAlpha=0.5}
   }
   n53 = Anim Dynamics - Bone: Dyn_skirt_side_r_01(ComponentPose=@n52, Alpha=1, ExternalForce="0,0,0") {
    angularDampingOverride=0.9
    angularSpringConstant=800
    bAngularSpring=true
    bGravityOverrideInSimSpace=true
    bOverrideAngularDamping=true
    bUseGravityOverride=true
    bUsePlanarLimit=false
    simulationSpace=RootRelative
    worldSpaceMinimumScale=0.022
    boundBone {boneName=Dyn_skirt_side_r_01}
    componentAppliedLinearAccClamp=<3000 3000 3000>
    componentLinearAccScale=<0.6 0.6 0.6>
    componentLinearVelScale=<0.6 0.6 0.6>
    gravityOverride {z=-980}
    physicsBodyDefinitions [{
      boundBone {boneName=Dyn_skirt_side_r_01}
      boxExtents=<12 12 12>
      constraintSetup {angularLimitsMax=<0 1 33>, angularLimitsMin=<0 -1 1>, angularTarget=<1 0 0>}
      localJointOffset=<0.4 0 0>
     }]
    simSpaceSettings {maxAngularAcceleration=3000, maxAngularVelocity=3000, simSpaceAngularAlpha=0.5}
   }
   n54 = Anim Dynamics - Bone: Dyn_skirt_side_r_02(ComponentPose=@n53, Alpha=1, ExternalForce="0,0,0") {
    angularDampingOverride=0.9
    angularSpringConstant=800
    bGravityOverrideInSimSpace=true
    bOverrideAngularDamping=true
    bUseGravityOverride=true
    bUsePlanarLimit=false
    simulationSpace=RootRelative
    worldSpaceMinimumScale=0.022
    boundBone {boneName=Dyn_skirt_side_r_02}
    componentAppliedLinearAccClamp=<3000 3000 3000>
    componentLinearAccScale=<0.6 0.6 0.6>
    componentLinearVelScale=<0.6 0.6 0.6>
    gravityOverride {z=-980}
    physicsBodyDefinitions [{
      boundBone {boneName=Dyn_skirt_side_r_02}
      boxExtents=<12 12 12>
      constraintSetup {angularLimitsMax=<0 1 33>, angularLimitsMin=<0 -1 1>, angularTarget=<1 0 0>}
      localJointOffset=<0.4 0 0>
     }]
    simSpaceSettings {maxAngularAcceleration=3000, maxAngularVelocity=3000, simSpaceAngularAlpha=0.5}
   }
   n55 = Anim Dynamics - Bone: Dyn_skirt_side_bk_r_01(ComponentPose=@n54, Alpha=0, ExternalForce="0,0,0") {
    angularDampingOverride=0.9
    angularSpringConstant=800
    bAngularSpring=true
    bGravityOverrideInSimSpace=true
    bOverrideAngularDamping=true
    bUseGravityOverride=true
    bUsePlanarLimit=false
    simulationSpace=RootRelative
    worldSpaceMinimumScale=0.022
    boundBone {boneName=Dyn_skirt_side_bk_r_01}
    componentAppliedLinearAccClamp=<3000 3000 3000>
    componentLinearAccScale=<0.6 0.6 0.6>
    componentLinearVelScale=<0.6 0.6 0.6>
    gravityOverride {z=-980}
    physicsBodyDefinitions [{
      boundBone {boneName=Dyn_skirt_side_bk_r_01}
      boxExtents=<12 12 12>
      constraintSetup {angularLimitsMax=<0 0 7>, angularLimitsMin=<0 0 1>, angularTarget=<1 0 0>}
      localJointOffset=<0.5 0 0>
     }]
    simSpaceSettings {maxAngularAcceleration=3000, maxAngularVelocity=3000, simSpaceAngularAlpha=0.5}
   }
   n56 = Anim Dynamics - Bone: Dyn_skirt_side_bk_r_02(ComponentPose=@n55, Alpha=0, ExternalForce="0,0,0") {
    angularDampingOverride=0.9
    angularSpringConstant=800
    bAngularSpring=true
    bGravityOverrideInSimSpace=true
    bOverrideAngularDamping=true
    bUseGravityOverride=true
    bUsePlanarLimit=false
    simulationSpace=RootRelative
    worldSpaceMinimumScale=0.022
    boundBone {boneName=Dyn_skirt_side_bk_r_02}
    componentAppliedLinearAccClamp=<3000 3000 3000>
    componentLinearAccScale=<0.6 0.6 0.6>
    componentLinearVelScale=<0.6 0.6 0.6>
    gravityOverride {z=-980}
    physicsBodyDefinitions [{
      boundBone {boneName=Dyn_skirt_side_bk_r_02}
      boxExtents=<12 12 12>
      constraintSetup {angularLimitsMax=<0 0 7>, angularLimitsMin=<0 0 1>, angularTarget=<1 0 0>}
      localJointOffset=<0.4 0 0>
     }]
    simSpaceSettings {maxAngularAcceleration=3000, maxAngularVelocity=3000, simSpaceAngularAlpha=0.5}
   }
   n57 = Anim Dynamics - Bone: Dyn_skirt_bk_r_01(ComponentPose=@n56, Alpha=0, ExternalForce="0,0,0") {
    angularDampingOverride=0.9
    angularSpringConstant=800
    bAngularSpring=true
    bGravityOverrideInSimSpace=true
    bOverrideAngularDamping=true
    bUseGravityOverride=true
    bUsePlanarLimit=false
    simulationSpace=RootRelative
    worldSpaceMinimumScale=0.022
    boundBone {boneName=Dyn_skirt_bk_r_01}
    componentAppliedLinearAccClamp=<3000 3000 3000>
    componentLinearAccScale=<0.6 0.6 0.6>
    componentLinearVelScale=<0.6 0.6 0.6>
    gravityOverride {z=-980}
    physicsBodyDefinitions [{
      boundBone {boneName=Dyn_skirt_bk_r_01}
      boxExtents=<12 12 12>
      constraintSetup {angularLimitsMax=<0 0 7>, angularLimitsMin=<0 0 1>, angularTarget=<1 0 0>}
      localJointOffset=<0.5 0 0>
     }]
    simSpaceSettings {maxAngularAcceleration=3000, maxAngularVelocity=3000, simSpaceAngularAlpha=0.5}
   }
   n58 = Anim Dynamics - Bone: Dyn_skirt_bk_r_02(ComponentPose=@n57, Alpha=0, ExternalForce="0,0,0") {
    angularDampingOverride=0.9
    angularSpringConstant=800
    bAngularSpring=true
    bGravityOverrideInSimSpace=true
    bOverrideAngularDamping=true
    bUseGravityOverride=true
    bUsePlanarLimit=false
    simulationSpace=RootRelative
    worldSpaceMinimumScale=0.022
    boundBone {boneName=Dyn_skirt_bk_r_02}
    componentAppliedLinearAccClamp=<3000 3000 3000>
    componentLinearAccScale=<0.6 0.6 0.6>
    componentLinearVelScale=<0.6 0.6 0.6>
    gravityOverride {z=-980}
    physicsBodyDefinitions [{
      boundBone {boneName=Dyn_skirt_bk_r_02}
      boxExtents=<12 12 12>
      constraintSetup {angularLimitsMax=<0 0 7>, angularLimitsMin=<0 0 1>, angularTarget=<1 0 0>}
      localJointOffset=<0.4 0 0>
     }]
    simSpaceSettings {maxAngularAcceleration=3000, maxAngularVelocity=3000, simSpaceAngularAlpha=0.5}
   }
   n59 = Anim Dynamics - Bone: Dyn_skirt_btw_bk_r_01(ComponentPose=@n58, Alpha=0, ExternalForce="0,0,0") {
    angularDampingOverride=0.9
    angularSpringConstant=800
    bAngularSpring=true
    bGravityOverrideInSimSpace=true
    bOverrideAngularDamping=true
    bUseGravityOverride=true
    bUsePlanarLimit=false
    simulationSpace=RootRelative
    worldSpaceMinimumScale=0.022
    boundBone {boneName=Dyn_skirt_btw_bk_r_01}
    componentAppliedLinearAccClamp=<3000 3000 3000>
    componentLinearAccScale=<1 1 1>
    componentLinearVelScale=<1 1 1>
    gravityOverride {z=-980}
    physicsBodyDefinitions [{
      boundBone {boneName=Dyn_skirt_btw_bk_r_01}
      boxExtents=<12 12 12>
      constraintSetup {angularLimitsMax=<0 0 7>, angularLimitsMin=<0 0 1>, angularTarget=<1 0 0>}
      localJointOffset=<0.5 0 0>
     }]
    simSpaceSettings {maxAngularAcceleration=3000, maxAngularVelocity=3000, simSpaceAngularAlpha=0.5}
   }
   n60 = Anim Dynamics - Bone: Dyn_skirt_btw_bk_r_02(ComponentPose=@n59, Alpha=0, ExternalForce="0,0,0") {
    angularDampingOverride=0.9
    angularSpringConstant=800
    bAngularSpring=true
    bGravityOverrideInSimSpace=true
    bOverrideAngularDamping=true
    bUseGravityOverride=true
    bUsePlanarLimit=false
    simulationSpace=RootRelative
    worldSpaceMinimumScale=0.022
    boundBone {boneName=Dyn_skirt_btw_bk_r_02}
    componentAppliedLinearAccClamp=<3000 3000 3000>
    componentLinearAccScale=<1 1 1>
    componentLinearVelScale=<1 1 1>
    gravityOverride {z=-980}
    physicsBodyDefinitions [{
      boundBone {boneName=Dyn_skirt_btw_bk_r_02}
      boxExtents=<12 12 12>
      constraintSetup {angularLimitsMax=<0 0 7>, angularLimitsMin=<0 0 1>, angularTarget=<1 0 0>}
      localJointOffset=<0.4 0 0>
     }]
    simSpaceSettings {maxAngularAcceleration=3000, maxAngularVelocity=3000, simSpaceAngularAlpha=0.5}
   }
   n61 = Anim Dynamics - Bone: Dyn_skirt_mid_bk_01(ComponentPose=@n60, Alpha=1, ExternalForce="0,0,0") {
    angularDampingOverride=0.9
    angularSpringConstant=800
    bAngularSpring=true
    bGravityOverrideInSimSpace=true
    bOverrideAngularDamping=true
    bUseGravityOverride=true
    bUsePlanarLimit=false
    simulationSpace=RootRelative
    worldSpaceMinimumScale=0.022
    boundBone {boneName=Dyn_skirt_mid_bk_01}
    componentAppliedLinearAccClamp=<3000 3000 3000>
    componentLinearAccScale=<0.6 0.6 0.6>
    componentLinearVelScale=<0.6 0.6 0.6>
    gravityOverride {z=-980}
    physicsBodyDefinitions [{
      boundBone {boneName=Dyn_skirt_mid_bk_01}
      boxExtents=<12 12 12>
      constraintSetup {angularLimitsMax=<0 0 7>, angularLimitsMin=<0 0 1>, angularTarget=<1 0 0>}
      localJointOffset=<0.5 0 0>
     }]
    simSpaceSettings {maxAngularAcceleration=3000, maxAngularVelocity=3000, simSpaceAngularAlpha=0.5}
   }
   n62 = Anim Dynamics - Bone: Dyn_skirt_mid_bk_02(ComponentPose=@n61, Alpha=1, ExternalForce="0,0,0") {
    angularDampingOverride=0.9
    angularSpringConstant=800
    bAngularSpring=true
    bGravityOverrideInSimSpace=true
    bOverrideAngularDamping=true
    bUseGravityOverride=true
    bUsePlanarLimit=false
    simulationSpace=RootRelative
    worldSpaceMinimumScale=0.022
    boundBone {boneName=Dyn_skirt_mid_bk_02}
    componentAppliedLinearAccClamp=<3000 3000 3000>
    componentLinearAccScale=<1 1 1>
    componentLinearVelScale=<1 1 1>
    gravityOverride {z=-980}
    physicsBodyDefinitions [{
      boundBone {boneName=Dyn_skirt_mid_bk_02}
      boxExtents=<12 12 12>
      constraintSetup {angularLimitsMax=<0 0 7>, angularLimitsMin=<0 0 1>, angularTarget=<1 0 0>}
      localJointOffset=<0.4 0 0>
     }]
    simSpaceSettings {maxAngularAcceleration=3000, maxAngularVelocity=3000, simSpaceAngularAlpha=0.5}
   }
   n63 = Anim Dynamics - Bone: Dyn_skirt_btw_bk_l_01(ComponentPose=@n62, Alpha=1, ExternalForce="0,0,0") {
    angularDampingOverride=0.9
    angularSpringConstant=800
    bAngularSpring=true
    bGravityOverrideInSimSpace=true
    bOverrideAngularDamping=true
    bUseGravityOverride=true
    bUsePlanarLimit=false
    simulationSpace=RootRelative
    worldSpaceMinimumScale=0.022
    boundBone {boneName=Dyn_skirt_btw_bk_l_01}
    componentAppliedLinearAccClamp=<3000 3000 3000>
    componentLinearAccScale=<0.6 0.6 0.6>
    componentLinearVelScale=<0.6 0.6 0.6>
    gravityOverride {z=-980}
    physicsBodyDefinitions [{
      boundBone {boneName=Dyn_skirt_btw_bk_l_01}
      boxExtents=<12 12 12>
      constraintSetup {angularLimitsMax=<0 0 7>, angularLimitsMin=<0 0 1>, angularTarget=<1 0 0>}
      localJointOffset=<0.5 0 0>
     }]
    simSpaceSettings {maxAngularAcceleration=3000, maxAngularVelocity=3000, simSpaceAngularAlpha=0.5}
   }
   n64 = Anim Dynamics - Bone: Dyn_skirt_btw_bk_l_02(ComponentPose=@n63, Alpha=1, ExternalForce="0,0,0") {
    angularDampingOverride=0.9
    angularSpringConstant=800
    bAngularSpring=true
    bGravityOverrideInSimSpace=true
    bOverrideAngularDamping=true
    bUseGravityOverride=true
    bUsePlanarLimit=false
    simulationSpace=RootRelative
    worldSpaceMinimumScale=0.022
    boundBone {boneName=Dyn_skirt_btw_bk_l_02}
    componentAppliedLinearAccClamp=<3000 3000 3000>
    componentLinearAccScale=<1 1 1>
    componentLinearVelScale=<1 1 1>
    gravityOverride {z=-980}
    physicsBodyDefinitions [{
      boundBone {boneName=Dyn_skirt_btw_bk_l_02}
      boxExtents=<12 12 12>
      constraintSetup {angularLimitsMax=<0 0 7>, angularLimitsMin=<0 0 1>, angularTarget=<1 0 0>}
      localJointOffset=<0.4 0 0>
     }]
    simSpaceSettings {maxAngularAcceleration=3000, maxAngularVelocity=3000, simSpaceAngularAlpha=0.5}
   }
   n65 = Anim Dynamics - Bone: Dyn_skirt_bk_l_01(ComponentPose=@n64, Alpha=1, ExternalForce="0,0,0") {
    angularDampingOverride=0.9
    angularSpringConstant=800
    bAngularSpring=true
    bGravityOverrideInSimSpace=true
    bOverrideAngularDamping=true
    bUseGravityOverride=true
    bUsePlanarLimit=false
    simulationSpace=RootRelative
    worldSpaceMinimumScale=0.022
    boundBone {boneName=Dyn_skirt_bk_l_01}
    componentAppliedLinearAccClamp=<3000 3000 3000>
    componentLinearAccScale=<0.6 0.6 0.6>
    componentLinearVelScale=<0.6 0.6 0.6>
    gravityOverride {z=-980}
    physicsBodyDefinitions [{
      boundBone {boneName=Dyn_skirt_bk_l_01}
      boxExtents=<12 12 12>
      constraintSetup {angularLimitsMax=<0 0 7>, angularLimitsMin=<0 0 1>, angularTarget=<1 0 0>}
      localJointOffset=<0.5 0 0>
     }]
    simSpaceSettings {maxAngularAcceleration=3000, maxAngularVelocity=3000, simSpaceAngularAlpha=0.5}
   }
   n66 = Anim Dynamics - Bone: Dyn_skirt_bk_l_02(ComponentPose=@n65, Alpha=1, ExternalForce="0,0,0") {
    angularDampingOverride=0.9
    angularSpringConstant=800
    bAngularSpring=true
    bGravityOverrideInSimSpace=true
    bOverrideAngularDamping=true
    bUseGravityOverride=true
    bUsePlanarLimit=false
    simulationSpace=RootRelative
    worldSpaceMinimumScale=0.022
    boundBone {boneName=Dyn_skirt_bk_l_02}
    componentAppliedLinearAccClamp=<3000 3000 3000>
    componentLinearAccScale=<0.6 0.6 0.6>
    componentLinearVelScale=<0.6 0.6 0.6>
    gravityOverride {z=-980}
    physicsBodyDefinitions [{
      boundBone {boneName=Dyn_skirt_bk_l_02}
      boxExtents=<12 12 12>
      constraintSetup {angularLimitsMax=<0 0 7>, angularLimitsMin=<0 0 1>, angularTarget=<1 0 0>}
      localJointOffset=<0.4 0 0>
     }]
    simSpaceSettings {maxAngularAcceleration=3000, maxAngularVelocity=3000, simSpaceAngularAlpha=0.5}
   }
   n67 = Anim Dynamics - Bone: Base_Sword_case(ComponentPose=@n66, Alpha=1, ExternalForce="0,0,0") {
    angularDampingOverride=0.9
    angularSpringConstant=800
    bAngularSpring=true
    bOverrideAngularDamping=true
    bUseGravityOverride=true
    bUsePlanarLimit=false
    simulationSpace=RootRelative
    boundBone {boneName=Base_Sword_case}
    componentAppliedLinearAccClamp=<3000 3000 3000>
    componentLinearAccScale=<0.6 0.6 0.6>
    componentLinearVelScale=<0.6 0.6 0.6>
    gravityOverride {z=-980}
    physicsBodyDefinitions [{
      boundBone {boneName=Base_Sword_case}
      boxExtents=<13 13 13>
      constraintSetup {coneAngle=1, angularLimitsMax=<0 2 0>, angularLimitsMin=<0 -2 -3>, angularTarget=<1 0 0>}
      localJointOffset=<0.5 0 0>
     }]
    simSpaceSettings {maxAngularAcceleration=3000, maxAngularVelocity=3000, simSpaceAngularAlpha=0.1}
   }
   Save cached pose 'Start Cache'(Pose=Component To Local(ComponentPose=@n67)) {cacheName=Start Cache}
   n68 = RigidBody!(
    ComponentPose=Local To Component(LocalPose=Use cached pose 'Start Cache'())
    Alpha
    OverrideWorldGravity="0,0,0"
    ExternalForce:Vector
    SimSpaceSettings=(WorldAlpha=1,VelocityScaleZ=0.7,DampingAlpha=0,MaxLinearVelocity=800,MaxAngularVelocity=8,MaxLinearAcceleration=3500,MaxAngularAcceleration=40,ExternalLinearDragV=(X=0.2,Y=0.2,Z=0.2),ExternalLinearVelocity=(X=0.2,Y=0.2,Z=0.2),ExternalAngularVelocity=(X=0,Y=0,Z=0))
   ) -> (Pose) {
    bClampLinearTranslationLimitToRefPose=true
    bOverrideWorldGravity=true
    overridePhysicsAsset=&/M_MED_BeaverStage_PhysicsAsset_NoBackBling
    simulationSpace=BaseBoneSpace
    baseBoneRef {boneName=root}
    componentAppliedLinearAccClamp=<3000 3000 3000>
    simSpaceSettings {
     dampingAlpha=0
     maxAngularAcceleration=40
     maxAngularVelocity=8
     maxLinearAcceleration=3500
     maxLinearVelocity=800
     velocityScaleZ=0.7
     worldAlpha=1
     externalLinearDragV=<0.2 0.2 0.2>
     externalLinearVelocity=<0.2 0.2 0.2>
    }
   }
   n69 = Component To Local(ComponentPose=@n68)
   n70 = RigidBody(
    ComponentPose=Local To Component(LocalPose=Use cached pose 'Start Cache'())
    OverrideWorldGravity="0,0,0"
    SimSpaceSettings=(WorldAlpha=1,VelocityScaleZ=0.7,DampingAlpha=0,MaxLinearVelocity=800,MaxAngularVelocity=8,MaxLinearAcceleration=3500,MaxAngularAcceleration=40,ExternalLinearDragV=(X=0.2,Y=0.2,Z=0.2),ExternalLinearVelocity=(X=0.2,Y=0.2,Z=0.2),ExternalAngularVelocity=(X=0,Y=0,Z=0))
   ) {
    bOverrideWorldGravity=true
    overridePhysicsAsset=&/M_MED_BeaverStage_PhysicsAsset_hasBackblink
    simulationSpace=BaseBoneSpace
    baseBoneRef {boneName=root}
    simSpaceSettings {
     dampingAlpha=0
     maxAngularAcceleration=40
     maxAngularVelocity=8
     maxLinearAcceleration=3500
     maxLinearVelocity=800
     velocityScaleZ=0.7
     worldAlpha=1
     externalLinearDragV=<0.2 0.2 0.2>
     externalLinearVelocity=<0.2 0.2 0.2>
    }
   }
   n71 = "Transform (Modify) Bone - Bone: attach_backpack"!(ComponentPose=@n70, bAlphaBoolEnabled:bool=True, Translation="0,-1,0", Rotation="0,0,5", Scale="1,1,1") -> (Pose) {
    alphaInputType=Bool
    rotationMode=BMM_Additive
    rotationSpace=BCS_BoneSpace
    translationMode=BMM_Additive
    translationSpace=BCS_BoneSpace
    alphaBoolBlend {blendInTime=0.1, blendOption=Linear, blendOutTime=0.1}
    boneToModify {boneName=attach_backpack}
    rotation {roll=5}
    translation {y=-1}
   }
   n72 = Component To Local(ComponentPose=@n71)
   n73 = Get IsWearingBackbling()
   n74 = "Transform (Modify) Bone - Bone: cape_mid_02_l"!(
    ComponentPose=Local To Component(LocalPose=Blend Poses by bool(BlendPose_0=@n72, BlendPose_1=@n69, BlendTime_0=0, BlendTime_1=0, bActiveValue=@n73))
    bAlphaBoolEnabled:bool=True
    Translation="0,0,0"
    Rotation="20,0,0"
    Scale="1,1,1"
   ) -> (Pose) {
    alphaInputType=Bool
    rotationMode=BMM_Additive
    rotationSpace=BCS_BoneSpace
    alphaBoolBlend {blendOption=Linear}
    boneToModify {boneName=cape_mid_02_l}
    rotation {pitch=20}
   }
   n75 = "Transform (Modify) Bone - Bone: cape_mid_02_r"!(ComponentPose=@n74, bAlphaBoolEnabled:bool=True, Translation="0,0,0", Rotation="20,0,0", Scale="1,1,1") -> (Pose) {
    alphaInputType=Bool
    rotationMode=BMM_Additive
    rotationSpace=BCS_BoneSpace
    alphaBoolBlend {blendOption=Linear}
    boneToModify {boneName=cape_mid_02_r}
    rotation {pitch=20}
   }
   n76 = "Transform (Modify) Bone - Bone: cape_03_l"!(ComponentPose=@n75, bAlphaBoolEnabled:bool=True, Translation="0,0,0", Rotation="20,0,0", Scale="1,1,1") -> (Pose) {
    alphaInputType=Bool
    rotationMode=BMM_Additive
    rotationSpace=BCS_BoneSpace
    alphaBoolBlend {blendOption=Linear}
    boneToModify {boneName=cape_03_l}
    rotation {pitch=20}
   }
   n77 = "Transform (Modify) Bone - Bone: cape_03_r"!(ComponentPose=@n76, bAlphaBoolEnabled:bool=True, Translation="0,0,0", Rotation="20,0,0", Scale="1,1,1") -> (Pose) {
    alphaInputType=Bool
    rotationMode=BMM_Additive
    rotationSpace=BCS_BoneSpace
    alphaBoolBlend {blendOption=Linear}
    boneToModify {boneName=cape_03_r}
    rotation {pitch=20}
   }
   Save cached pose 'Rban Cache'(Pose=Blend Poses by bool(
     BlendPose_0=Use cached pose 'Start Cache'()
     BlendPose_1=Component To Local(ComponentPose=@n77)
     BlendTime_0=0
     BlendTime_1=0
     bActiveValue=Get bClothEnabled()
    )) {cacheName=Rban Cache}
   n78 = Anim Dynamics - Bone: cape_04_l!(
    ComponentPose=Local To Component(LocalPose=Use cached pose 'Rban Cache'())
    Alpha=1
    GravityOverride:Vector="0,0,-980"
    ExternalForce="0,0,0"
   ) -> (Pose) {
    angularDampingOverride=1
    angularSpringConstant=10000
    bAngularSpring=true
    bOverrideAngularDamping=true
    bUseGravityOverride=true
    gravityScale=0.35
    simulationSpace=BoneRelative
    alphaBoolBlend {blendOption=Linear}
    boundBone {boneName=cape_04_l}
    componentAppliedLinearAccClamp=<3000 3000 3000>
    componentLinearAccScale=<2 2 2>
    componentLinearVelScale=<2 2 2>
    gravityOverride {z=-980}
    physicsBodyDefinitions [{
      boundBone {boneName=cape_04_l}
      boxExtents=<10 40 40>
      constraintSetup {angularLimitsMax=<0 0 15>, angularLimitsMin=<0 -40 -15>, angularTarget=<1 0 0>}
      localJointOffset=<5 0 0>
     }]
    relativeSpaceBone {boneName=root}
   }
   n79 = Anim Dynamics - Bone: cape_05_l!(ComponentPose=@n78, Alpha=1, GravityOverride:Vector="0,0,-980", ExternalForce="0,0,0") -> (Pose) {
    angularDampingOverride=1
    angularSpringConstant=10000
    bAngularSpring=true
    bOverrideAngularDamping=true
    bUseGravityOverride=true
    gravityScale=0.35
    simulationSpace=BoneRelative
    alphaBoolBlend {blendOption=Linear}
    boundBone {boneName=cape_05_l}
    componentAppliedLinearAccClamp=<3000 3000 3000>
    componentLinearAccScale=<2 2 2>
    componentLinearVelScale=<2 2 2>
    gravityOverride {z=-980}
    physicsBodyDefinitions [{
      boundBone {boneName=cape_05_l}
      boxExtents=<10 40 40>
      constraintSetup {angularLimitsMax=<0 0 15>, angularLimitsMin=<0 -40 -15>, angularTarget=<1 0 0>}
      localJointOffset=<5 0 0>
     }]
    relativeSpaceBone {boneName=root}
   }
   n80 = Anim Dynamics - Bone: cape_06_l!(ComponentPose=@n79, Alpha=1, GravityOverride:Vector="0,0,-980", ExternalForce="0,0,0") -> (Pose) {
    angularDampingOverride=1
    angularSpringConstant=10000
    bAngularSpring=true
    bOverrideAngularDamping=true
    bUseGravityOverride=true
    gravityScale=0.35
    simulationSpace=BoneRelative
    alphaBoolBlend {blendOption=Linear}
    boundBone {boneName=cape_06_l}
    componentAppliedLinearAccClamp=<3000 3000 3000>
    componentLinearAccScale=<2 2 2>
    componentLinearVelScale=<2 2 2>
    gravityOverride {z=-980}
    physicsBodyDefinitions [{
      boundBone {boneName=cape_06_l}
      boxExtents=<10 40 40>
      constraintSetup {angularLimitsMax=<0 5 15>, angularLimitsMin=<0 -40 -15>, angularTarget=<1 0 0>}
      localJointOffset=<5 0 0>
     }]
    relativeSpaceBone {boneName=root}
   }
   n81 = Anim Dynamics - Bone: cape_04_r!(ComponentPose=@n80, Alpha=1, GravityOverride:Vector="0,0,-980", ExternalForce="0,0,0") -> (Pose) {
    angularDampingOverride=1
    angularSpringConstant=10000
    bAngularSpring=true
    bOverrideAngularDamping=true
    bUseGravityOverride=true
    gravityScale=0.35
    simulationSpace=BoneRelative
    alphaBoolBlend {blendOption=Linear}
    boundBone {boneName=cape_04_r}
    componentAppliedLinearAccClamp=<3000 3000 3000>
    componentLinearAccScale=<2 2 2>
    componentLinearVelScale=<2 2 2>
    gravityOverride {z=-980}
    physicsBodyDefinitions [{
      boundBone {boneName=cape_04_r}
      boxExtents=<10 40 40>
      constraintSetup {angularLimitsMax=<0 0 15>, angularLimitsMin=<0 -40 -15>, angularTarget=<1 0 0>}
      localJointOffset=<5 0 0>
     }]
    relativeSpaceBone {boneName=root}
   }
   n82 = Anim Dynamics - Bone: cape_05_r!(ComponentPose=@n81, Alpha=1, GravityOverride:Vector="0,0,-980", ExternalForce="0,0,0") -> (Pose) {
    angularDampingOverride=1
    angularSpringConstant=10000
    bAngularSpring=true
    bOverrideAngularDamping=true
    bUseGravityOverride=true
    gravityScale=0.35
    simulationSpace=BoneRelative
    alphaBoolBlend {blendOption=Linear}
    boundBone {boneName=cape_05_r}
    componentAppliedLinearAccClamp=<3000 3000 3000>
    componentLinearAccScale=<2 2 2>
    componentLinearVelScale=<2 2 2>
    gravityOverride {z=-980}
    physicsBodyDefinitions [{
      boundBone {boneName=cape_05_r}
      boxExtents=<10 40 40>
      constraintSetup {angularLimitsMax=<0 0 15>, angularLimitsMin=<0 -40 -15>, angularTarget=<1 0 0>}
      localJointOffset=<5 0 0>
     }]
    relativeSpaceBone {boneName=root}
   }
   n83 = Anim Dynamics - Bone: cape_06_r!(ComponentPose=@n82, Alpha=1, GravityOverride:Vector="0,0,-980", ExternalForce="0,0,0") -> (Pose) {
    angularDampingOverride=1
    angularSpringConstant=10000
    bAngularSpring=true
    bOverrideAngularDamping=true
    bUseGravityOverride=true
    gravityScale=0.35
    simulationSpace=BoneRelative
    alphaBoolBlend {blendOption=Linear}
    boundBone {boneName=cape_06_r}
    componentAppliedLinearAccClamp=<3000 3000 3000>
    componentLinearAccScale=<2 2 2>
    componentLinearVelScale=<2 2 2>
    gravityOverride {z=-980}
    physicsBodyDefinitions [{
      boundBone {boneName=cape_06_r}
      boxExtents=<10 40 40>
      constraintSetup {angularLimitsMax=<0 5 15>, angularLimitsMin=<0 -40 -15>, angularTarget=<1 0 0>}
      localJointOffset=<5 0 0>
     }]
    relativeSpaceBone {boneName=root}
   }
   n84 = Anim Dynamics - Bone: cape_mid_03_l!(ComponentPose=@n83, Alpha=1, GravityOverride:Vector="0,0,-980", ExternalForce="0,0,0") -> (Pose) {
    angularDampingOverride=1
    angularSpringConstant=10000
    bAngularSpring=true
    bOverrideAngularDamping=true
    bUseGravityOverride=true
    gravityScale=0.35
    simulationSpace=BoneRelative
    alphaBoolBlend {blendOption=Linear}
    boundBone {boneName=cape_mid_03_l}
    componentAppliedLinearAccClamp=<3000 3000 3000>
    componentLinearAccScale=<2 2 2>
    componentLinearVelScale=<2 2 2>
    gravityOverride {z=-980}
    physicsBodyDefinitions [{
      boundBone {boneName=cape_mid_03_l}
      boxExtents=<10 40 40>
      constraintSetup {angularLimitsMax=<0 0 15>, angularLimitsMin=<0 -40 -15>, angularTarget=<1 0 0>}
      localJointOffset=<5 0 0>
     }]
    relativeSpaceBone {boneName=root}
   }
   n85 = Anim Dynamics - Bone: cape_mid_04_l!(ComponentPose=@n84, Alpha=1, GravityOverride:Vector="0,0,-980", ExternalForce="0,0,0") -> (Pose) {
    angularDampingOverride=1
    angularSpringConstant=10000
    bAngularSpring=true
    bOverrideAngularDamping=true
    bUseGravityOverride=true
    gravityScale=0.35
    simulationSpace=BoneRelative
    alphaBoolBlend {blendOption=Linear}
    boundBone {boneName=cape_mid_04_l}
    componentAppliedLinearAccClamp=<3000 3000 3000>
    componentLinearAccScale=<2 2 2>
    componentLinearVelScale=<2 2 2>
    gravityOverride {z=-980}
    physicsBodyDefinitions [{
      boundBone {boneName=cape_mid_04_l}
      boxExtents=<10 40 40>
      constraintSetup {angularLimitsMax=<0 0 15>, angularLimitsMin=<0 -40 -15>, angularTarget=<1 0 0>}
      localJointOffset=<5 0 0>
     }]
    relativeSpaceBone {boneName=root}
   }
   n86 = Anim Dynamics - Bone: cape_mid_05_l!(ComponentPose=@n85, Alpha=1, GravityOverride:Vector="0,0,-980", ExternalForce="0,0,0") -> (Pose) {
    angularDampingOverride=1
    angularSpringConstant=10000
    bAngularSpring=true
    bOverrideAngularDamping=true
    bUseGravityOverride=true
    gravityScale=0.35
    simulationSpace=BoneRelative
    alphaBoolBlend {blendOption=Linear}
    boundBone {boneName=cape_mid_05_l}
    componentAppliedLinearAccClamp=<3000 3000 3000>
    componentLinearAccScale=<2 2 2>
    componentLinearVelScale=<2 2 2>
    gravityOverride {z=-980}
    physicsBodyDefinitions [{
      boundBone {boneName=cape_mid_05_l}
      boxExtents=<10 40 40>
      constraintSetup {angularLimitsMax=<0 5 15>, angularLimitsMin=<0 -40 -15>, angularTarget=<1 0 0>}
      localJointOffset=<5 0 0>
     }]
    relativeSpaceBone {boneName=root}
   }
   n87 = Anim Dynamics - Bone: cape_mid_03_r!(ComponentPose=@n86, Alpha=1, GravityOverride:Vector="0,0,-980", ExternalForce="0,0,0") -> (Pose) {
    angularDampingOverride=1
    angularSpringConstant=10000
    bAngularSpring=true
    bOverrideAngularDamping=true
    bUseGravityOverride=true
    gravityScale=0.35
    simulationSpace=BoneRelative
    alphaBoolBlend {blendOption=Linear}
    boundBone {boneName=cape_mid_03_r}
    componentAppliedLinearAccClamp=<3000 3000 3000>
    componentLinearAccScale=<2 2 2>
    componentLinearVelScale=<2 2 2>
    gravityOverride {z=-980}
    physicsBodyDefinitions [{
      boundBone {boneName=cape_mid_03_r}
      boxExtents=<10 40 40>
      constraintSetup {angularLimitsMax=<0 0 15>, angularLimitsMin=<0 -40 -15>, angularTarget=<1 0 0>}
      localJointOffset=<5 0 0>
     }]
    relativeSpaceBone {boneName=root}
   }
   n88 = Anim Dynamics - Bone: cape_mid_04_r!(ComponentPose=@n87, Alpha=1, GravityOverride:Vector="0,0,-980", ExternalForce="0,0,0") -> (Pose) {
    angularDampingOverride=1
    angularSpringConstant=10000
    bAngularSpring=true
    bOverrideAngularDamping=true
    bUseGravityOverride=true
    gravityScale=0.35
    simulationSpace=BoneRelative
    alphaBoolBlend {blendOption=Linear}
    boundBone {boneName=cape_mid_04_r}
    componentAppliedLinearAccClamp=<3000 3000 3000>
    componentLinearAccScale=<2 2 2>
    componentLinearVelScale=<2 2 2>
    gravityOverride {z=-980}
    physicsBodyDefinitions [{
      boundBone {boneName=cape_mid_04_r}
      boxExtents=<10 40 40>
      constraintSetup {angularLimitsMax=<0 0 15>, angularLimitsMin=<0 -40 -15>, angularTarget=<1 0 0>}
      localJointOffset=<5 0 0>
     }]
    relativeSpaceBone {boneName=root}
   }
   n89 = Anim Dynamics - Bone: cape_mid_05_r!(ComponentPose=@n88, Alpha=1, GravityOverride:Vector="0,0,-980", ExternalForce="0,0,0") -> (Pose) {
    angularDampingOverride=1
    angularSpringConstant=10000
    bAngularSpring=true
    bOverrideAngularDamping=true
    bUseGravityOverride=true
    gravityScale=0.35
    simulationSpace=BoneRelative
    alphaBoolBlend {blendOption=Linear}
    boundBone {boneName=cape_mid_05_r}
    componentAppliedLinearAccClamp=<3000 3000 3000>
    componentLinearAccScale=<2 2 2>
    componentLinearVelScale=<2 2 2>
    gravityOverride {z=-980}
    physicsBodyDefinitions [{
      boundBone {boneName=cape_mid_05_r}
      boxExtents=<10 40 40>
      constraintSetup {angularLimitsMax=<0 5 15>, angularLimitsMin=<0 -40 -15>, angularTarget=<1 0 0>}
      localJointOffset=<5 0 0>
     }]
    relativeSpaceBone {boneName=root}
   }
   n90 = "Transform (Modify) Bone - Bone: cape_mid_02_l"(ComponentPose=@n89, bAlphaBoolEnabled=True, Translation="0,0,0", Rotation="5,0,0", Scale="1,1,1") {
    alphaInputType=Bool
    rotationMode=BMM_Additive
    rotationSpace=BCS_BoneSpace
    alphaBoolBlend {blendOption=Linear}
    boneToModify {boneName=cape_mid_02_l}
    rotation {pitch=5}
   }
   n91 = "Transform (Modify) Bone - Bone: cape_mid_02_r"(ComponentPose=@n90, bAlphaBoolEnabled=True, Translation="0,0,0", Rotation="5,0,0", Scale="1,1,1") {
    alphaInputType=Bool
    rotationMode=BMM_Additive
    rotationSpace=BCS_BoneSpace
    alphaBoolBlend {blendOption=Linear}
    boneToModify {boneName=cape_mid_02_r}
    rotation {pitch=5}
   }
   n92 = "Transform (Modify) Bone - Bone: cape_03_l"(ComponentPose=@n91, bAlphaBoolEnabled=True, Translation="0,0,0", Rotation="5,0,0", Scale="1,1,1") {
    alphaInputType=Bool
    rotationMode=BMM_Additive
    rotationSpace=BCS_BoneSpace
    alphaBoolBlend {blendOption=Linear}
    boneToModify {boneName=cape_03_l}
    rotation {pitch=5}
   }
   n93 = "Transform (Modify) Bone - Bone: cape_03_r"(ComponentPose=@n92, bAlphaBoolEnabled=True, Translation="0,0,0", Rotation="5,0,0", Scale="1,1,1") {
    alphaInputType=Bool
    rotationMode=BMM_Additive
    rotationSpace=BCS_BoneSpace
    alphaBoolBlend {blendOption=Linear}
    boneToModify {boneName=cape_03_r}
    rotation {pitch=5}
   }
   n94 = "Transform (Modify) Bone - Bone: attach_backpack"(ComponentPose=@n93, bAlphaBoolEnabled=True, Translation="0,-1,0", Rotation="0,0,5", Scale="1,1,1") {
    alphaInputType=Bool
    rotationMode=BMM_Additive
    rotationSpace=BCS_BoneSpace
    translationMode=BMM_Additive
    translationSpace=BCS_BoneSpace
    alphaBoolBlend {blendInTime=0.1, blendOption=Linear, blendOutTime=0.1}
    boneToModify {boneName=attach_backpack}
    rotation {roll=5}
    translation {y=-1}
   }
   n95 = Anim Dynamics - Bone: cape_02_l!(
    ComponentPose=Local To Component(LocalPose=Use cached pose 'Rban Cache'())
    Alpha=1
    GravityOverride:Vector="0,0,-980"
    ExternalForce="0,0,0"
   ) -> (Pose) {
    angularDampingOverride=1
    angularSpringConstant=10000
    bAngularSpring=true
    bOverrideAngularDamping=true
    bUseGravityOverride=true
    gravityScale=0.35
    simulationSpace=BoneRelative
    alphaBoolBlend {blendOption=Linear}
    boundBone {boneName=cape_02_l}
    componentAppliedLinearAccClamp=<3000 3000 3000>
    componentLinearAccScale=<2 2 2>
    componentLinearVelScale=<2 2 2>
    gravityOverride {z=-980}
    physicsBodyDefinitions [{
      boundBone {boneName=cape_02_l}
      boxExtents=<10 40 40>
      constraintSetup {angularLimitsMax=<0 0 15>, angularLimitsMin=<0 -40 -15>, angularTarget=<1 0 0>}
      localJointOffset=<5 0 0>
     }]
    relativeSpaceBone {boneName=root}
   }
   n96 = Anim Dynamics - Bone: cape_04_l(ComponentPose=@n95, Alpha=1, GravityOverride="0,0,-980", ExternalForce="0,0,0") {
    angularDampingOverride=1
    angularSpringConstant=10000
    bAngularSpring=true
    bOverrideAngularDamping=true
    bUseGravityOverride=true
    gravityScale=0.35
    simulationSpace=BoneRelative
    alphaBoolBlend {blendOption=Linear}
    boundBone {boneName=cape_04_l}
    componentAppliedLinearAccClamp=<3000 3000 3000>
    componentLinearAccScale=<2 2 2>
    componentLinearVelScale=<2 2 2>
    gravityOverride {z=-980}
    physicsBodyDefinitions [{
      boundBone {boneName=cape_04_l}
      boxExtents=<10 40 40>
      constraintSetup {angularLimitsMax=<0 0 15>, angularLimitsMin=<0 -40 -15>, angularTarget=<1 0 0>}
      localJointOffset=<5 0 0>
     }]
    relativeSpaceBone {boneName=root}
   }
   n97 = Anim Dynamics - Bone: cape_06_l(ComponentPose=@n96, Alpha=1, GravityOverride="0,0,-980", ExternalForce="0,0,0") {
    angularDampingOverride=1
    angularSpringConstant=10000
    bAngularSpring=true
    bOverrideAngularDamping=true
    bUseGravityOverride=true
    gravityScale=0.35
    simulationSpace=BoneRelative
    alphaBoolBlend {blendOption=Linear}
    boundBone {boneName=cape_06_l}
    componentAppliedLinearAccClamp=<3000 3000 3000>
    componentLinearAccScale=<2 2 2>
    componentLinearVelScale=<2 2 2>
    gravityOverride {z=-980}
    physicsBodyDefinitions [{
      boundBone {boneName=cape_06_l}
      boxExtents=<10 40 40>
      constraintSetup {angularLimitsMax=<0 5 15>, angularLimitsMin=<0 -40 -15>, angularTarget=<1 0 0>}
      localJointOffset=<5 0 0>
     }]
    relativeSpaceBone {boneName=root}
   }
   n98 = Anim Dynamics - Bone: cape_02_r!(ComponentPose=@n97, Alpha=1, GravityOverride:Vector="0,0,-980", ExternalForce="0,0,0") -> (Pose) {
    angularDampingOverride=1
    angularSpringConstant=10000
    bAngularSpring=true
    bOverrideAngularDamping=true
    bUseGravityOverride=true
    gravityScale=0.35
    simulationSpace=BoneRelative
    alphaBoolBlend {blendOption=Linear}
    boundBone {boneName=cape_02_r}
    componentAppliedLinearAccClamp=<3000 3000 3000>
    componentLinearAccScale=<2 2 2>
    componentLinearVelScale=<2 2 2>
    gravityOverride {z=-980}
    physicsBodyDefinitions [{
      boundBone {boneName=cape_02_r}
      boxExtents=<10 40 40>
      constraintSetup {angularLimitsMax=<0 0 15>, angularLimitsMin=<0 -40 -15>, angularTarget=<1 0 0>}
      localJointOffset=<5 0 0>
     }]
    relativeSpaceBone {boneName=root}
   }
   n99 = Anim Dynamics - Bone: cape_04_r(ComponentPose=@n98, Alpha=1, GravityOverride="0,0,-980", ExternalForce="0,0,0") {
    angularDampingOverride=1
    angularSpringConstant=10000
    bAngularSpring=true
    bOverrideAngularDamping=true
    bUseGravityOverride=true
    gravityScale=0.35
    simulationSpace=BoneRelative
    alphaBoolBlend {blendOption=Linear}
    boundBone {boneName=cape_04_r}
    componentAppliedLinearAccClamp=<3000 3000 3000>
    componentLinearAccScale=<2 2 2>
    componentLinearVelScale=<2 2 2>
    gravityOverride {z=-980}
    physicsBodyDefinitions [{
      boundBone {boneName=cape_04_r}
      boxExtents=<10 40 40>
      constraintSetup {angularLimitsMax=<0 0 15>, angularLimitsMin=<0 -40 -15>, angularTarget=<1 0 0>}
      localJointOffset=<5 0 0>
     }]
    relativeSpaceBone {boneName=root}
   }
   n100 = Anim Dynamics - Bone: cape_06_r(ComponentPose=@n99, Alpha=1, GravityOverride="0,0,-980", ExternalForce="0,0,0") {
    angularDampingOverride=1
    angularSpringConstant=10000
    bAngularSpring=true
    bOverrideAngularDamping=true
    bUseGravityOverride=true
    gravityScale=0.35
    simulationSpace=BoneRelative
    alphaBoolBlend {blendOption=Linear}
    boundBone {boneName=cape_06_r}
    componentAppliedLinearAccClamp=<3000 3000 3000>
    componentLinearAccScale=<2 2 2>
    componentLinearVelScale=<2 2 2>
    gravityOverride {z=-980}
    physicsBodyDefinitions [{
      boundBone {boneName=cape_06_r}
      boxExtents=<10 40 40>
      constraintSetup {angularLimitsMax=<0 5 15>, angularLimitsMin=<0 -40 -15>, angularTarget=<1 0 0>}
      localJointOffset=<5 0 0>
     }]
    relativeSpaceBone {boneName=root}
   }
   n101 = Anim Dynamics - Bone: cape_mid_01_l!(ComponentPose=@n100, Alpha=1, GravityOverride:Vector="0,0,-980", ExternalForce="0,0,0") -> (Pose) {
    angularDampingOverride=1
    angularSpringConstant=10000
    bAngularSpring=true
    bOverrideAngularDamping=true
    bUseGravityOverride=true
    gravityScale=0.35
    simulationSpace=BoneRelative
    alphaBoolBlend {blendOption=Linear}
    boundBone {boneName=cape_mid_01_l}
    componentAppliedLinearAccClamp=<3000 3000 3000>
    componentLinearAccScale=<2 2 2>
    componentLinearVelScale=<2 2 2>
    gravityOverride {z=-980}
    physicsBodyDefinitions [{
      boundBone {boneName=cape_mid_01_l}
      boxExtents=<10 40 40>
      constraintSetup {angularLimitsMax=<0 0 15>, angularLimitsMin=<0 -40 -15>, angularTarget=<1 0 0>}
      localJointOffset=<5 0 0>
     }]
    relativeSpaceBone {boneName=root}
   }
   n102 = Anim Dynamics - Bone: cape_mid_03_l(ComponentPose=@n101, Alpha=1, GravityOverride="0,0,-980", ExternalForce="0,0,0") {
    angularDampingOverride=1
    angularSpringConstant=10000
    bAngularSpring=true
    bOverrideAngularDamping=true
    bUseGravityOverride=true
    gravityScale=0.35
    simulationSpace=BoneRelative
    alphaBoolBlend {blendOption=Linear}
    boundBone {boneName=cape_mid_03_l}
    componentAppliedLinearAccClamp=<3000 3000 3000>
    componentLinearAccScale=<2 2 2>
    componentLinearVelScale=<2 2 2>
    gravityOverride {z=-980}
    physicsBodyDefinitions [{
      boundBone {boneName=cape_mid_03_l}
      boxExtents=<10 40 40>
      constraintSetup {angularLimitsMax=<0 0 15>, angularLimitsMin=<0 -40 -15>, angularTarget=<1 0 0>}
      localJointOffset=<5 0 0>
     }]
    relativeSpaceBone {boneName=root}
   }
   n103 = Anim Dynamics - Bone: cape_mid_05_l(ComponentPose=@n102, Alpha=1, GravityOverride="0,0,-980", ExternalForce="0,0,0") {
    angularDampingOverride=1
    angularSpringConstant=10000
    bAngularSpring=true
    bOverrideAngularDamping=true
    bUseGravityOverride=true
    gravityScale=0.35
    simulationSpace=BoneRelative
    alphaBoolBlend {blendOption=Linear}
    boundBone {boneName=cape_mid_05_l}
    componentAppliedLinearAccClamp=<3000 3000 3000>
    componentLinearAccScale=<2 2 2>
    componentLinearVelScale=<2 2 2>
    gravityOverride {z=-980}
    physicsBodyDefinitions [{
      boundBone {boneName=cape_mid_05_l}
      boxExtents=<10 40 40>
      constraintSetup {angularLimitsMax=<0 5 15>, angularLimitsMin=<0 -40 -15>, angularTarget=<1 0 0>}
      localJointOffset=<5 0 0>
     }]
    relativeSpaceBone {boneName=root}
   }
   n104 = Anim Dynamics - Bone: cape_mid_01_r!(ComponentPose=@n103, Alpha=1, GravityOverride:Vector="0,0,-980", ExternalForce="0,0,0") -> (Pose) {
    angularDampingOverride=1
    angularSpringConstant=10000
    bAngularSpring=true
    bOverrideAngularDamping=true
    bUseGravityOverride=true
    gravityScale=0.35
    simulationSpace=BoneRelative
    alphaBoolBlend {blendOption=Linear}
    boundBone {boneName=cape_mid_01_r}
    componentAppliedLinearAccClamp=<3000 3000 3000>
    componentLinearAccScale=<2 2 2>
    componentLinearVelScale=<2 2 2>
    gravityOverride {z=-980}
    physicsBodyDefinitions [{
      boundBone {boneName=cape_mid_01_r}
      boxExtents=<10 40 40>
      constraintSetup {angularLimitsMax=<0 0 15>, angularLimitsMin=<0 -40 -15>, angularTarget=<1 0 0>}
      localJointOffset=<5 0 0>
     }]
    relativeSpaceBone {boneName=root}
   }
   n105 = Anim Dynamics - Bone: cape_mid_03_r(ComponentPose=@n104, Alpha=1, GravityOverride="0,0,-980", ExternalForce="0,0,0") {
    angularDampingOverride=1
    angularSpringConstant=10000
    bAngularSpring=true
    bOverrideAngularDamping=true
    bUseGravityOverride=true
    gravityScale=0.35
    simulationSpace=BoneRelative
    alphaBoolBlend {blendOption=Linear}
    boundBone {boneName=cape_mid_03_r}
    componentAppliedLinearAccClamp=<3000 3000 3000>
    componentLinearAccScale=<2 2 2>
    componentLinearVelScale=<2 2 2>
    gravityOverride {z=-980}
    physicsBodyDefinitions [{
      boundBone {boneName=cape_mid_03_r}
      boxExtents=<10 40 40>
      constraintSetup {angularLimitsMax=<0 0 15>, angularLimitsMin=<0 -40 -15>, angularTarget=<1 0 0>}
      localJointOffset=<5 0 0>
     }]
    relativeSpaceBone {boneName=root}
   }
   n106 = Anim Dynamics - Bone: cape_mid_05_r(ComponentPose=@n105, Alpha=1, GravityOverride="0,0,-980", ExternalForce="0,0,0") {
    angularDampingOverride=1
    angularSpringConstant=10000
    bAngularSpring=true
    bOverrideAngularDamping=true
    bUseGravityOverride=true
    gravityScale=0.35
    simulationSpace=BoneRelative
    alphaBoolBlend {blendOption=Linear}
    boundBone {boneName=cape_mid_05_r}
    componentAppliedLinearAccClamp=<3000 3000 3000>
    componentLinearAccScale=<2 2 2>
    componentLinearVelScale=<2 2 2>
    gravityOverride {z=-980}
    physicsBodyDefinitions [{
      boundBone {boneName=cape_mid_05_r}
      boxExtents=<10 40 40>
      constraintSetup {angularLimitsMax=<0 5 15>, angularLimitsMin=<0 -40 -15>, angularTarget=<1 0 0>}
      localJointOffset=<5 0 0>
     }]
    relativeSpaceBone {boneName=root}
   }
   Save cached pose 'AnimDyn Cache'(Pose=Blend Poses by bool(
     BlendPose_0=Component To Local(ComponentPose=@n94)
     BlendPose_1=Component To Local(ComponentPose=@n106)
     bActiveValue=Get IsWearingBackbling()
    )) {cacheName=AnimDyn Cache}
  ]}
 EventGraph {nodes [
   n1 = Event BlueprintUpdateAnimation()
   > Set DeltaTimeLocal(DeltaTimeLocal:double=@n1.DeltaTimeX) -> (Output_Get:double)
   > n2 = Sequence!(execute) -> (then_0, then_1, then_2:Exec)
   > n3 = PlayerMovement(AnimInstance=Self-Reference(), XDotProductCoefficient=1, YDotProductCoefficient=1, ZDotProductCoefficient=1)
   > n4 = Set Is Right Left Movement(Is Right Left Movement:double=@n3) -> (Output_Get:double)
   > n5 = Set Is Forward Backward Movement(Is Forward Backward Movement:double=@n3.YMovement) -> (Output_Get:double)
   n6 = "Absolute (Float)"(A=@n4)
   n7 = MapRangeClamped(Value=@n5, InRangeA=0, InRangeB=1)
   > Set BwdMovementTrail(BwdMovementTrail:double=MapRangeClamped(Value=float + float(A=@n6, B=@n7) {is=title K2Node_CommutativeAssociativeBinaryOperator}, InRangeA=0, InRangeB=1.4)) -> (Output_Get:double)
   > n8 = Branch(Condition=Get bIsSwimSprinting() -> (bIsSwimSprinting:bool))
   > Set TrailControllerRot(TrailControllerRot:Rotator="45,0,0") -> (Output_Get:Rotator)
   n9 = Branch(execute=@n8.else, Condition=Get bIsSkyDiving())
   > n10 = Branch(Condition=Get bIsParachuteOpened())
   > Set TrailControllerRot(TrailControllerRot="45,0, 0")
   n11 = Branch(execute=@n10.else, Condition="float > float"(A=Get PawnVelocityZ(), B=50) {is=title K2Node_CallFunction})
   > Set BwdMovementTrail(BwdMovementTrail=float * float(A=Get BwdMovementTrail(), B=0))
   > Set TrailControllerRot(TrailControllerRot="45,0, 0")
   Set TrailControllerRot(execute=@n11.else, TrailControllerRot="40,0,0")
   Set TrailControllerRot(execute=@n9.else, TrailControllerRot="15,0,0")
   > Set BwdMovementTrail(BwdMovementTrail=float * float(A=Get BwdMovementTrail(), B=1))
   n12 = Branch(execute=@n2.then_1, Condition=Get bClothEnabled())
   > "Cloth Controller Function (Parameter Asset) (Nativized)"!(
    execute
    CharacterPartAnimInstance:Custom Character Part Anim Instance Object Reference=Self-Reference()
    ClothingAssetName:Name=Base_Cape_01
    ClothCoreSettings:"Cloth Core Settings Structure (by ref)"=Get Cloth Core Settings() -> (Cloth Core Settings:Cloth Core Settings Structure)
    WindGustLocalSettings:"Wind Gust Settings Structure (by ref)"=Get Wind Gust Local Settings() -> (Wind Gust Local Settings:Wind Gust Settings Structure)
    ClothUpAxis:Vector="1,0,0"
    LinVelScaleFwdBwdClamp:float=0
    LinVelScaleUpDwnClamp:float=0.7
    DeltaTime:float=Get DeltaTimeLocal() -> (DeltaTimeLocal:double)
    GameTimeInSeconds:float=GetGameTimeInSeconds()
    StateLogicLayerName:Name=none
    Parameters:Fort Character Dynamics Parameters Object Reference=&/M_MED_BeaverStage_Parameters_New
   ) -> (
    then
    OutCurrentState:Name=None
    OutClothParameters:Cloth Parameters Structure
    OutClothCoreSettings:Cloth Core Settings Structure
    OutWindGustSettings:Wind Gust Parameters Structure
    OutWindGustLocalSettings:Wind Gust Settings Structure
    OutWindVelocity:Vector
    OutLinearVelVector:Vector
   )
   n13 = Branch(execute=@n12.else, Condition=Get IsWearingBackbling())
   > n14 = "Gravity Override State Function (Parameter Asset) (Nativized)"(
    CharacterPartAnimInstance=Self-Reference()
    GameTimeInSeconds=GetGameTimeInSeconds()
    Parameters=/BRCosmetics/Accessories/FORT_Capes/Cape_M_MED_WealthLamb_Fate/Meshes/M_MED_WealthLamb_Fate_Cape_Parameters
   )
   > Set Noncloth Gravity Override(Noncloth Gravity Override:Vector=@n14) -> (Output_Get:Vector)
   n15 = "Gravity Override State Function (Parameter Asset) (Nativized)"(
    execute=@n13.else
    CharacterPartAnimInstance=Self-Reference()
    GameTimeInSeconds=GetGameTimeInSeconds()
    Parameters=&M_MED_BeaverStage_Parameters_New
   )
   > Set Noncloth Gravity Override(Noncloth Gravity Override=@n15)
   > Set RBAN Settings Out(RBAN Settings Out=@n15.OutRigidBodyAnimNodeParameters)
   > Set DivingDown(DivingDown:bool=Contains(SearchIn="To String (Name)"(InName=@n15.OutCurrentState), Substring=DivingDown)) -> (Output_Get:bool)
   Event OnEndJumping()
   > Set Is Jumping(Is Jumping:bool=false) -> (Output_Get:bool)
   Event OnBeginBaseballSliding() {is=title K2Node_Event}
   > Set BaseballSliding(BaseballSliding:double=1) -> (Output_Get:double)
   Event OnBeginSkydiving()
   > Set Sky(Sky:bool=true) -> (Output_Get:bool)
   Event OnBeginJumping()
   > Set Is Jumping(Is Jumping=true)
   Event OnCharacterCustomizationLoaded()
   > n16 = Sequence()
   n17 = CheckForBackbling()
   > Set IsWearingBackbling(IsWearingBackbling:bool=OR Boolean!(A=@n17, B=false, C:bool=@n17.bIsBackblingLong) -> (ReturnValue)) -> (Output_Get:bool)
   n18 = Branch(execute=@n16.then_1, Condition=Get IsWearingBackbling())
   n19 = GetPartSkeletalMeshComponent(self=Self-Reference())
   > Set ClothBlendWeight!(execute, ClothBlendWeight:float=0, self:Skeletal Mesh Component Object Reference=@n19) -> (then, Output_Get:float)
   > Set bDisableClothSimulation!(execute, bDisableClothSimulation:bool=true, self:Skeletal Mesh Component Object Reference=@n19) -> (then, Output_Get:bool)
   Set ClothBlendWeight(execute=@n18.else, ClothBlendWeight=1, self=@n19)
   > Set bDisableClothSimulation(bDisableClothSimulation=false, self=@n19)
   Event OnEndSkydiving()
   > Set Sky(Sky=false)
   Event BlueprintInitializeAnimation()
   > Set Skel_Mesh(Skel_Mesh=GetOwnerSkeletalMeshComponent())
   Event OnEndBaseballSliding() {is=title K2Node_Event}
   > Set BaseballSliding(BaseballSliding=0)
   Event OnBeginPickaxeSwing() {is=title K2Node_Event}
   > Set Picaxeswing(Picaxeswing:double=1) -> (Output_Get:double)
   Event OnEndPickaxeSwing() {is=title K2Node_Event}
   > Set Picaxeswing(Picaxeswing=0)
   TryGetPawnOwner()
   Branch(Condition=Get bIsSkydiveDiveMode())
   Set TrailControllerRot(TrailControllerRot="40,0,0")
  ]}
]
variables [
 BaseballSliding {type=number}
 BwdMovementTrail {type=number}
 Cloth Core Settings {
  type=ClothCoreSettings
  default {
   collisionThickness=1
   dampingCoefficient=0.1
   frictionCoefficient=0.3
   localDampingCoefficient=0
   numberIterations=2
   numberSubsteps=1
   selfCollisionThickness=0
   animDriveDamping=<0 1>
   animDriveStiffness=<0 1>
   areaStiffness=<1 1>
   bendingStiffness=<1 1.05>
   edgeStiffness=<1 1>
   tetherStiffness=<1 1>
  }
 }
 DeltaTimeLocal {type=number}
 DivingDown {type=boolean}
 Is Forward Backward Movement {type=number}
 Is Jumping {type=boolean}
 Is Right Left Movement {type=number}
 IsbeaverPickaxe {type=boolean}
 IsWearingBackbling {type=boolean}
 Noncloth Gravity Override {type=Vector}
 Picaxeswing {type=number}
 RBAN Settings Out {type=FortRigidBodyAnimNodeParameters}
 Skel_Mesh {type=SkeletalMeshComponent}
 Sky {type=boolean}
 TrailControllerRot {type=Rotator}
 Wind Gust Local Settings {type=WindGustSettings, default {elapsedTime=0, finalGustStrengthLocal=0, gustDuration=0, gustStrength=0, timeBetweenGusts=0}}
]
