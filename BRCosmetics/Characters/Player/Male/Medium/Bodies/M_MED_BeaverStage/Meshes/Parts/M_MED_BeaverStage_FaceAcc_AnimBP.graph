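# Post-pose AnimBP for the BeaverStage face accessory, copying the parent mesh's pose and adding spring-driven Brush bone motion.
# Without backbling, the nine-bone tail uses chained AnimDynamics plus additive Tail_base offsets, increased when crouch-targeting or skydiving.
# With backbling, a cached alternative constrains Tail_01–08 to weighted head, neck and upper-spine bones.
# The EventGraph also computes head-rotation values, which the AnimGraph never reads.
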AnimBlueprint /BRCosmetics/Characters/Player/Male/Medium/Bodies/M_MED_BeaverStage/Meshes/Parts/M_MED_BeaverStage_FaceAcc_AnimBP
target_skeleton=&/M_MED_BeaverStage_FaceAcc_Skeleton
graphs [
 AnimGraph {nodes [
   n1 = Anim Dynamics - Bone: Tail_01(ComponentPose=Local To Component(LocalPose=Use cached pose 'Start Cache'()), Alpha=1, ExternalForce="0,0,0") {
    angularDampingOverride=0.99
    angularSpringConstant=300
    bAngularSpring=true
    bOverrideAngularDamping=true
    bUseGravityOverride=true
    bUsePlanarLimit=false
    simulationSpace=RootRelative
    boundBone {boneName=Tail_01}
    componentAppliedLinearAccClamp=<3000 3000 3500>
    componentLinearAccScale=<0.1 0.1 0.8>
    componentLinearVelScale=<0.1 0.1 0.8>
    gravityOverride {z=-980}
    physicsBodyDefinitions [{
      boundBone {boneName=Tail_01}
      boxExtents=<11 11 11>
      constraintSetup {angularLimitsMax=<1.5 4 8>, angularLimitsMin=<-1.5 -4 -1.5>, angularTarget=<1 0 0>}
      localJointOffset=<0.66 0 0>
     }]
    simSpaceSettings {maxAngularAcceleration=5000, maxAngularVelocity=5000, simSpaceAngularAlpha=0.22}
   }
   n2 = Anim Dynamics - Bone: Tail_02(ComponentPose=@n1, Alpha=1, ExternalForce="0,0,0") {
    angularDampingOverride=0.99
    angularSpringConstant=300
    bAngularSpring=true
    bOverrideAngularDamping=true
    bUseGravityOverride=true
    bUsePlanarLimit=false
    simulationSpace=RootRelative
    boundBone {boneName=Tail_02}
    componentAppliedLinearAccClamp=<3000 3000 3500>
    componentLinearAccScale=<0.1 0.1 0.8>
    componentLinearVelScale=<0.1 0.1 0.8>
    gravityOverride {z=-980}
    physicsBodyDefinitions [{
      boundBone {boneName=Tail_02}
      boxExtents=<11 11 11>
      constraintSetup {angularLimitsMax=<1.5 4 8>, angularLimitsMin=<-1.5 -4 -1>, angularTarget=<1 0 0>}
      localJointOffset=<0.66 0 0>
     }]
    simSpaceSettings {maxAngularAcceleration=5000, maxAngularVelocity=5000, simSpaceAngularAlpha=0.22}
   }
   n3 = Anim Dynamics - Bone: Tail_03(ComponentPose=@n2, Alpha=1, ExternalForce="0,0,0") {
    angularDampingOverride=0.99
    angularSpringConstant=300
    bAngularSpring=true
    bOverrideAngularDamping=true
    bUseGravityOverride=true
    bUsePlanarLimit=false
    simulationSpace=RootRelative
    boundBone {boneName=Tail_03}
    componentAppliedLinearAccClamp=<3000 3000 3500>
    componentLinearAccScale=<0.1 0.1 0.8>
    componentLinearVelScale=<0.1 0.1 0.8>
    gravityOverride {z=-980}
    physicsBodyDefinitions [{
      boundBone {boneName=Tail_03}
      boxExtents=<11 11 11>
      constraintSetup {angularLimitsMax=<1.5 4 8>, angularLimitsMin=<-1.5 -4 1>, angularTarget=<1 0 0>}
      localJointOffset=<0.66 0 0>
     }]
    simSpaceSettings {maxAngularAcceleration=5000, maxAngularVelocity=5000, simSpaceAngularAlpha=0.22}
   }
   n4 = Anim Dynamics - Bone: Tail_04(ComponentPose=@n3, Alpha=1, ExternalForce="0,0,0") {
    angularDampingOverride=0.99
    angularSpringConstant=300
    bAngularSpring=true
    bOverrideAngularDamping=true
    bUseGravityOverride=true
    bUsePlanarLimit=false
    simulationSpace=RootRelative
    boundBone {boneName=Tail_04}
    componentAppliedLinearAccClamp=<3000 3000 3500>
    componentLinearAccScale=<0.1 0.1 0.8>
    componentLinearVelScale=<0.1 0.1 0.8>
    gravityOverride {z=-980}
    physicsBodyDefinitions [{
      boundBone {boneName=Tail_04}
      boxExtents=<11 11 11>
      constraintSetup {angularLimitsMax=<1.5 4 8>, angularLimitsMin=<-1.5 -4 1>, angularTarget=<1 0 0>}
      localJointOffset=<0.66 0 0>
     }]
    simSpaceSettings {maxAngularAcceleration=5000, maxAngularVelocity=5000, simSpaceAngularAlpha=0.22}
   }
   n5 = Anim Dynamics - Bone: Tail_05(ComponentPose=@n4, Alpha=1, ExternalForce="0,0,0") {
    angularDampingOverride=0.99
    angularSpringConstant=300
    bAngularSpring=true
    bOverrideAngularDamping=true
    bUseGravityOverride=true
    bUsePlanarLimit=false
    simulationSpace=RootRelative
    boundBone {boneName=Tail_05}
    componentAppliedLinearAccClamp=<3000 3000 3500>
    componentLinearAccScale=<0.1 0.1 0.8>
    componentLinearVelScale=<0.1 0.1 0.8>
    gravityOverride {z=-980}
    physicsBodyDefinitions [{
      boundBone {boneName=Tail_05}
      boxExtents=<11 11 11>
      constraintSetup {angularLimitsMax=<1.5 4 8>, angularLimitsMin=<-1.5 -4 1>, angularTarget=<1 0 0>}
      localJointOffset=<0.66 0 0>
     }]
    simSpaceSettings {maxAngularAcceleration=5000, maxAngularVelocity=5000, simSpaceAngularAlpha=0.22}
   }
   n6 = Anim Dynamics - Bone: Tail_06(ComponentPose=@n5, Alpha=1, ExternalForce="0,0,0") {
    angularDampingOverride=0.99
    angularSpringConstant=300
    bAngularSpring=true
    bOverrideAngularDamping=true
    bUseGravityOverride=true
    bUsePlanarLimit=false
    simulationSpace=RootRelative
    boundBone {boneName=Tail_06}
    componentAppliedLinearAccClamp=<3000 3000 3500>
    componentLinearAccScale=<0.1 0.1 0.8>
    componentLinearVelScale=<0.1 0.1 0.8>
    gravityOverride {z=-980}
    physicsBodyDefinitions [{
      boundBone {boneName=Tail_06}
      boxExtents=<11 11 11>
      constraintSetup {angularLimitsMax=<1.5 4 8>, angularLimitsMin=<-1.5 -4 1>, angularTarget=<1 0 0>}
      localJointOffset=<0.66 0 0>
     }]
    simSpaceSettings {maxAngularAcceleration=5000, maxAngularVelocity=5000, simSpaceAngularAlpha=0.22}
   }
   n7 = Anim Dynamics - Bone: Tail_07(ComponentPose=@n6, Alpha=1, ExternalForce="0,0,0") {
    angularDampingOverride=0.99
    angularSpringConstant=300
    bAngularSpring=true
    bOverrideAngularDamping=true
    bUseGravityOverride=true
    bUsePlanarLimit=false
    simulationSpace=RootRelative
    boundBone {boneName=Tail_07}
    componentAppliedLinearAccClamp=<3000 3000 3500>
    componentLinearAccScale=<0.1 0.1 0.8>
    componentLinearVelScale=<0.1 0.1 0.8>
    gravityOverride {z=-980}
    physicsBodyDefinitions [{
      boundBone {boneName=Tail_07}
      boxExtents=<11 11 11>
      constraintSetup {angularLimitsMax=<1.5 4 8>, angularLimitsMin=<-1.5 -4 1>, angularTarget=<1 0 0>}
      localJointOffset=<0.66 0 0>
     }]
    simSpaceSettings {maxAngularAcceleration=5000, maxAngularVelocity=5000, simSpaceAngularAlpha=0.22}
   }
   n8 = Anim Dynamics - Bone: Tail_08(ComponentPose=@n7, Alpha=1, ExternalForce="0,0,0") {
    angularDampingOverride=0.99
    angularSpringConstant=300
    bAngularSpring=true
    bOverrideAngularDamping=true
    bUseGravityOverride=true
    bUsePlanarLimit=false
    simulationSpace=RootRelative
    boundBone {boneName=Tail_08}
    componentAppliedLinearAccClamp=<3000 3000 3500>
    componentLinearAccScale=<0.1 0.1 0.8>
    componentLinearVelScale=<0.1 0.1 0.8>
    gravityOverride {z=-980}
    physicsBodyDefinitions [{
      boundBone {boneName=Tail_08}
      boxExtents=<11 11 11>
      constraintSetup {angularLimitsMax=<1.5 4 8>, angularLimitsMin=<-1.5 -4 1>, angularTarget=<1 0 0>}
      localJointOffset=<0.66 0 0>
     }]
    simSpaceSettings {maxAngularAcceleration=5000, maxAngularVelocity=5000, simSpaceAngularAlpha=0.22}
   }
   n9 = Anim Dynamics - Bone: Tail_09(ComponentPose=@n8, Alpha=1, ExternalForce="0,0,0") {
    angularDampingOverride=0.99
    angularSpringConstant=300
    bAngularSpring=true
    bOverrideAngularDamping=true
    bUseGravityOverride=true
    bUsePlanarLimit=false
    simulationSpace=RootRelative
    boundBone {boneName=Tail_09}
    componentAppliedLinearAccClamp=<3000 3000 3500>
    componentLinearAccScale=<0.1 0.1 0.8>
    componentLinearVelScale=<0.1 0.1 0.8>
    gravityOverride {z=-980}
    physicsBodyDefinitions [{
      boundBone {boneName=Tail_09}
      boxExtents=<11 11 11>
      constraintSetup {angularLimitsMax=<1.5 4 8>, angularLimitsMin=<-1.5 -4 1>, angularTarget=<1 0 0>}
      localJointOffset=<0.66 0 0>
     }]
    simSpaceSettings {maxAngularAcceleration=5000, maxAngularVelocity=5000, simSpaceAngularAlpha=0.22}
   }
   n10 = "Transform (Modify) Bone - Bone: Tail_base"(
    ComponentPose=@n9
    Alpha=1
    Translation=Get Translation() -> (Translation:Vector)
    Rotation=Get Rotation() -> (Rotation:Rotator)
    Scale="1,1,1"
   ) {
    rotationMode=BMM_Additive
    rotationSpace=BCS_BoneSpace
    translationMode=BMM_Additive
    translationSpace=BCS_BoneSpace
    boneToModify {boneName=Tail_base}
   }
   n11 = "Transform (Modify) Bone - Bone: Tail_base"(ComponentPose=@n10, Alpha=1, Translation="0.5,2,0", Rotation="0,20,0", Scale="1,1,1") {
    rotationMode=BMM_Additive
    rotationSpace=BCS_BoneSpace
    scaleMode=BMM_Additive
    scaleSpace=BCS_BoneSpace
    translationMode=BMM_Additive
    translationSpace=BCS_BoneSpace
    boneToModify {boneName=Tail_base}
    rotation {yaw=20}
    translation=<0.5 2>
   }
   Output Pose(Result=Blend Poses by bool(
     BlendPose_0=Use cached pose 'Constraint cache'()
     BlendPose_1=Component To Local(ComponentPose=@n11)
     bActiveValue=Get IsWearingBackbling() -> (IsWearingBackbling:bool)
    )) {layerGroup=None}
   n12 = Anim Dynamics - Bone: Brush_fr(ComponentPose=Local To Component(LocalPose=Copy Pose From Mesh() {bUseAttachedParent=true}), Alpha=1, ExternalForce="0,0,0") {
    angularDampingOverride=0.9
    angularSpringConstant=800
    bAngularSpring=true
    bOverrideAngularDamping=true
    bUseGravityOverride=true
    bUsePlanarLimit=false
    simulationSpace=RootRelative
    boundBone {boneName=Brush_fr}
    componentAppliedLinearAccClamp=<3000 3000 3000>
    componentLinearAccScale=<0.3 0.3 0.3>
    componentLinearVelScale=<0.3 0.3 0.3>
    gravityOverride {z=-980}
    physicsBodyDefinitions [{
      boundBone {boneName=Brush_fr}
      boxExtents=<12 12 12>
      constraintSetup {angularLimitsMax=<0 1 5>, angularLimitsMin=<0 -1 -5>, angularTarget=<1 0 0>}
      localJointOffset=<0.3 0 0>
     }]
    simSpaceSettings {maxAngularAcceleration=5000, maxAngularVelocity=5000, simSpaceAngularAlpha=0.22}
   }
   n13 = Anim Dynamics - Bone: Brush_mid(ComponentPose=@n12, Alpha=1, ExternalForce="0,0,0") {
    angularDampingOverride=0.9
    angularSpringConstant=800
    bOverrideAngularDamping=true
    bUseGravityOverride=true
    bUsePlanarLimit=false
    simulationSpace=RootRelative
    boundBone {boneName=Brush_mid}
    componentAppliedLinearAccClamp=<3000 3000 3000>
    componentLinearAccScale=<0.3 0.3 0.3>
    componentLinearVelScale=<0.3 0.3 0.3>
    gravityOverride {z=-980}
    physicsBodyDefinitions [{
      boundBone {boneName=Brush_mid}
      boxExtents=<12 12 12>
      constraintSetup {angularLimitsMax=<0 1 5>, angularLimitsMin=<0 -1 -5>, angularTarget=<1 0 0>}
      localJointOffset=<0.3 0 0>
     }]
    simSpaceSettings {maxAngularAcceleration=5000, maxAngularVelocity=5000, simSpaceAngularAlpha=0.22}
   }
   n14 = Anim Dynamics - Bone: Brush_bk(ComponentPose=@n13, Alpha=1, ExternalForce="0,0,0") {
    angularDampingOverride=0.9
    angularSpringConstant=800
    bOverrideAngularDamping=true
    bUseGravityOverride=true
    bUsePlanarLimit=false
    simulationSpace=RootRelative
    boundBone {boneName=Brush_bk}
    componentAppliedLinearAccClamp=<3000 3000 3000>
    componentLinearAccScale=<0.3 0.3 0.3>
    componentLinearVelScale=<0.3 0.3 0.3>
    gravityOverride {z=-980}
    physicsBodyDefinitions [{
      boundBone {boneName=Brush_bk}
      boxExtents=<12 12 12>
      constraintSetup {angularLimitsMax=<0 1 5>, angularLimitsMin=<0 -1 -5>, angularTarget=<1 0 0>}
      localJointOffset=<0.3 0 0>
     }]
    simSpaceSettings {maxAngularAcceleration=5000, maxAngularVelocity=5000, simSpaceAngularAlpha=0.22}
   }
   Save cached pose 'Start Cache'(Pose=Component To Local(ComponentPose=@n14)) {cacheName=Start Cache}
   n15 = "Constraint - Tail_01 (PP)"!(ComponentPose=Local To Component(LocalPose=Use cached pose 'Start Cache'()), Alpha=1, ConstraintWeights_0=0.5, ConstraintWeights_1:float=0.5) -> (Pose) {boneToModify {boneName=Tail_01}, constraintSetup [{targetBone {boneName=head}}, {targetBone {boneName=neck_02}}], constraintWeights [1,1]}
   n16 = "Constraint - Tail_02 (PP)"!(ComponentPose=@n15, Alpha=1, ConstraintWeights_0=0.5, ConstraintWeights_1:float=0.5) -> (Pose) {boneToModify {boneName=Tail_02}, constraintSetup [{targetBone {boneName=head}}, {targetBone {boneName=neck_02}}], constraintWeights [1,1]}
   n17 = "Constraint - Tail_03 (PPP)"!(ComponentPose=@n16, Alpha=1, ConstraintWeights_0=0.4, ConstraintWeights_1:float=0.5, ConstraintWeights_2:float=0.1) -> (Pose) {
    boneToModify {boneName=Tail_03}
    constraintSetup [{targetBone {boneName=head}}, {targetBone {boneName=neck_02}}, {targetBone {boneName=neck_01}}]
    constraintWeights [1,1,1]
   }
   n18 = "Constraint - Tail_04 (PPP)"!(ComponentPose=@n17, Alpha=1, ConstraintWeights_0=0.3, ConstraintWeights_1:float=0.5, ConstraintWeights_2:float=0.2) -> (Pose) {
    boneToModify {boneName=Tail_04}
    constraintSetup [{targetBone {boneName=head}}, {targetBone {boneName=neck_02}}, {targetBone {boneName=neck_01}}]
    constraintWeights [1,1,1]
   }
   n19 = "Constraint - Tail_05 (PPT)"!(ComponentPose=@n18, Alpha=1, ConstraintWeights_0=0.2, ConstraintWeights_1:float=0.5, ConstraintWeights_2:float=0.3) -> (Pose) {
    boneToModify {boneName=Tail_05}
    constraintSetup [{targetBone {boneName=head}}, {targetBone {boneName=neck_02}}, {transformType=Translation, targetBone {boneName=neck_01}}]
    constraintWeights [1,1,1]
   }
   n20 = "Constraint - Tail_05 (PPP)"!(ComponentPose=@n19, Alpha=1, ConstraintWeights_0=0.1, ConstraintWeights_1:float=0.4, ConstraintWeights_2:float=0.5) -> (Pose) {
    boneToModify {boneName=Tail_05}
    constraintSetup [{targetBone {boneName=head}}, {targetBone {boneName=neck_02}}, {targetBone {boneName=neck_01}}]
    constraintWeights [1,1,1]
   }
   n21 = "Constraint - Tail_06 (PPP)"!(ComponentPose=@n20, Alpha=1, ConstraintWeights_0=0.1, ConstraintWeights_1:float=0.3, ConstraintWeights_2:float=0.6) -> (Pose) {
    boneToModify {boneName=Tail_06}
    constraintSetup [{targetBone {boneName=spine_05}}, {targetBone {boneName=neck_02}}, {targetBone {boneName=neck_01}}]
    constraintWeights [1,1,1]
   }
   n22 = "Constraint - Tail_06 (PPP)"(ComponentPose=@n21, Alpha=1, ConstraintWeights_0=0.3, ConstraintWeights_1=0, ConstraintWeights_2=0.7) {
    boneToModify {boneName=Tail_06}
    constraintSetup [{targetBone {boneName=spine_05}}, {targetBone {boneName=neck_02}}, {targetBone {boneName=neck_01}}]
    constraintWeights [1,1,1]
   }
   n23 = "Constraint - Tail_07 (PPP)"!(ComponentPose=@n22, Alpha=1, ConstraintWeights_0=0.4, ConstraintWeights_1:float=0, ConstraintWeights_2:float=0.6) -> (Pose) {
    boneToModify {boneName=Tail_07}
    constraintSetup [{targetBone {boneName=spine_05}}, {targetBone {boneName=neck_02}}, {targetBone {boneName=neck_01}}]
    constraintWeights [1,1,1]
   }
   n24 = "Constraint - Tail_08 (PPP)"!(ComponentPose=@n23, Alpha=1, ConstraintWeights_0=0.6, ConstraintWeights_1:float=0, ConstraintWeights_2:float=0.4) -> (Pose) {
    boneToModify {boneName=Tail_08}
    constraintSetup [{targetBone {boneName=spine_05}}, {targetBone {boneName=neck_02}}, {targetBone {boneName=neck_01}}]
    constraintWeights [1,1,1]
   }
   Save cached pose 'Constraint cache'(Pose=Component To Local(ComponentPose=@n24)) {cacheName=Constraint cache}
  ]}
 EventGraph {nodes [
   n1 = Event BlueprintUpdateAnimation()
   n2 = BreakTransform(InTransform=GetDeltaTransformFromRefPose(self=GetOwnerSkeletalMeshComponent(), BoneName=spine_05, BaseName=head))
   n3 = BreakRotator(InRot=@n2.Rotation)
   > Set Head Rotate z(Head Rotate z:double=MapRangeClamped(Value=@n3.Yaw, InRangeA=6, InRangeB=40)) -> (Output_Get:double)
   > n4 = Sequence()
   n5 = Get bIsCrouching()
   n6 = Get bIsTargeting()
   > n7 = Branch(Condition=OR Boolean(A=AND Boolean(A=@n5, B=@n6), B=Get bIsSkyDiving()))
   > Set Rotation(Rotation:Rotator="0,45,0") -> (Output_Get:Rotator)
   > Set Translation(Translation:Vector="4,0,0") -> (Output_Get:Vector)
   Set Rotation(execute=@n7.else, Rotation="0,20,0")
   > Set Translation(Translation="2,0,0")
   n8 = MapRangeClamped(Value=@n3, InRangeA=25, InRangeB=45)
   n9 = MapRangeClamped(Value=@n3, InRangeA=-25, InRangeB=-45)
   n10 = Set Head Rotate x(
    execute=@n4.then_1
    Head Rotate x:double=FInterpTo(Current=Get Head Rotate x() -> (Head Rotate x:double), Target=float + float(A=@n9, B=@n8), DeltaTime=@n1.DeltaTimeX, InterpSpeed=14)
   ) -> (Output_Get:double)
   n11 = CheckForBackbling()
   > Set IsWearingBackbling(IsWearingBackbling:bool=OR Boolean!(A=@n11, B=false, C:bool=@n11.bIsBackblingLong) -> (ReturnValue)) -> (Output_Get:bool)
   n12 = Get bIsPlayingEmote()
   > n13 = Branch(Condition=AND Boolean(A=Get bIsInFrontEnd(), B=NOT Boolean(A=@n12)))
   > Set Head Rotate x(Head Rotate x=0.15)
   Set Head Rotate x(execute=@n13.else, Head Rotate x=@n10)
   Reroute Node(InputPin:Rotator=@n2.Rotation) -> (OutputPin:Rotator)
   Get bIsSkydiveFloating()
  ]}
]
variables [
 Head Rotate x {type=number}
 Head Rotate z {type=number}
 IsWearingBackbling {type=boolean}
 Rotation {type=Rotator, default=<rot 0 40 0>}
 Translation {type=Vector, default=<2 0 0>}
]
